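# Idle Niagara effect for the CoatCheck (GingerPesto) backpack, drawing persistent mesh-particle wings and glows on its animated bones.
# Wing_L/R burst feather particles, using partition indices to pick each one's bone, offset, rotation and scale; Wing_R mirrors X scale.
# BlueGlow_L/R follow the wing root bones; all L emitters sample _r bones and vice versa.
# Dynamic material parameters add random variation, with some values switched by front-end preview and platform.
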NiagaraSystem /BRCosmetics/Accessories/FORT_Backpacks/Backpack_F_MED_CoatCheck/FX/NS_Backpack_GingerPesto_Idle
dependencies {
 usedDataInterfaces [
  &NiagaraDataInterfaceArrayFloat3
  &NiagaraDataInterfacePlatformSet
  &NiagaraDataInterfaceSkeletalMesh
 ]
 usedDynamicInputs []
 usedModules [&SystemState]
}
emitters {
 BlueGlow_L {
  emitterUpdateScript [
   &EmitterState(Life Cycle Mode=::NewEnumerator1, Loop Behavior=::NewEnumerator1, Loop Duration Mode=::NewEnumerator1, UseLoopDelay=false)
   &SpawnBurst_Instantaneous(Use Spawn Probability=false, UseLoopCountLimit=false)
  ]
  particleSpawnScript [
   &InitializeParticle(Lifetime=5, Mesh Scale=<3 2.62 2.62>, Mesh Scale Mode=::NewEnumerator4, UsePositionOffset=false, Write Material Random=true)
   &InitialMeshOrientation(Mesh Orientation Mode=::NewEnumerator2, Rotation=<0 0.25 0>, Use Rotation Vector=true)
   &InitialMeshOrientation(Mesh Orientation Mode=::NewEnumerator2, Rotation=<0 0 0.36>, Use Rotation Vector=true) {module=InitialMeshOrientation001}
   /SetVariables(Particles.Size002:Vector3f=<3.5 2.8 2.8>) {isSetParameters=true}
   /SetVariables(Particles.Color002=&MultiplyLinearColor(Linear Color=<rgba 1.5 2 0 1>, Scale Factor=&MakeVector4_FromFloat(Float=0.7))) {isSetParameters=true}
   &DynamicMaterialParameters(
    Index 0 Param 1=&RandomRangeFloat(Minimum=1.5, Maximum=2)
    Index 0 Param 2=&RandomRangeFloat(Minimum=0.5, Maximum=0.7)
    Index 0 Param 3=&RandomRangeFloat(Minimum=0.02, Maximum=-0.02)
    Index 0 Param 4=&RandomRangeFloat(Minimum=-0.3, Maximum=-0.7)
    Index 2 Param 1=&RandomRangeFloat(Minimum=2, Maximum=4)
    Index 2 Param 2=&RandomRangeFloat(Minimum=0.5, Maximum=0.8)
    Index 2 Param 3=&RandomRangeFloat(Minimum=-0.01, Maximum=0.01)
    Index 2 Param 4=&RandomRangeFloat(Minimum=-0.8, Maximum=-1)
    Index 3 Param 1=&Multiply_Float(
     A=&MakeCustomFloatFromBool(Bool=$User.bIsFrontEndPreview, True Float=0.2, False Float=0.28)
     B=&FN_PickFloatDependOnPlatform(TargetPlatformActive=1, TargetPlatformNoActive=0.8)
    )
    Index 3 Param 2=&Multiply_Float(
     A=&MakeCustomFloatFromBool(Bool=$User.bIsFrontEndPreview, True Float=5, False Float=9)
     B=&FN_PickFloatDependOnPlatform(TargetPlatformActive=1, TargetPlatformNoActive=0.8)
    )
    Module.Index0Param0Write=true
    Module.Index0Param1Write=true
    Module.Index0Param2Write=true
    Module.Index0Param3Write=true
    Module.Index2Param0Write=true
    Module.Index2Param1Write=true
    Module.Index2Param2Write=true
    Module.Index2Param3Write=true
    Module.Index3Param0Write=true
    Module.Index3Param1Write=true
    Module.Index3Param2Write=true
    Module.Index3Param3Write=true
    Write Parameter Index 2=true
    Write Parameter Index 3=true
   )
  ]
  particleUpdateScript [
   &ParticleState(Kill Particles When Lifetime Has Elapsed=false, Let Infinitely Lived Particles Die When Emitter Deactivates=true)
   &ScaleMeshSize(Scale Factor=&SelectVectorFromArray(
     Direct Array Index=$Particles.PartitionParticles.Partition
     Vector Selection Array=DataInterface() {propertyValues {FloatData [<1 1 1>, <1 1 1>, <1 1 1>, <1 1 1>, <0.8 0.8 0.5>, <0.7 0.7 0.4>, <0.6 0.6 0.3>]}}
    ))
   &SkeletalMeshLocation(
    Bone Sampling Mode=::NewEnumerator4
    MaskSpawnGroup=false
    Skeletal Mesh=DataInterface() {propertyValues {
      DefaultMesh=null
      ExcludeBoneName=root
      OptionalComponentClass=null
      PreviewMesh=/BRCosmetics/Accessories/FORT_Backpacks/Backpack_F_MED_CoatCheck/Meshes/F_MED_CoatCheck_Pack
      SkinningMode=SkinOnTheFly
      SoftSourceActor=null
      SourceMode=Default
      UvSetIndex=0
      WholeMeshLOD=-1
      bExcludeBone=false
      bReadDeformedGeometry=true
      bRequireCurrentFrameData=true
      ComponentTags []
      FilteredBones [dyn_wing01_r]
      FilteredSockets []
      MeshUserParameter {parameter {name=None, typeDefHandle {registeredTypeIndex=1}, varData []}}
      SamplingRegions []
     }}
   )
  ]
  rendererClasses [&NiagaraMeshRendererProperties]
  renderers [
   {rendererClass=&NiagaraMeshRendererProperties, rendererIndex=0}
   {rendererClass=&NiagaraMeshRendererProperties, rendererIndex=1}
  ]
 }
 BlueGlow_R {
  emitterUpdateScript [
   &EmitterState(Life Cycle Mode=::NewEnumerator1, Loop Behavior=::NewEnumerator1, Loop Duration Mode=::NewEnumerator1, UseLoopDelay=false)
   &SpawnBurst_Instantaneous(Use Spawn Probability=false, UseLoopCountLimit=false)
  ]
  particleSpawnScript [
   &InitializeParticle(Lifetime=5, Mesh Scale=<3 2.62 2.62>, Mesh Scale Mode=::NewEnumerator4, UsePositionOffset=false, Write Material Random=true)
   &InitialMeshOrientation(Mesh Orientation Mode=::NewEnumerator2, Rotation=<0 0.24 0>, Use Rotation Vector=true)
   &InitialMeshOrientation(Mesh Orientation Mode=::NewEnumerator2, Rotation=<0 0 -0.142>, Use Rotation Vector=true) {module=InitialMeshOrientation001}
   /SetVariables(Particles.Size002:Vector3f=<3.5 2.8 2.8>) {isSetParameters=true}
   /SetVariables(Particles.Color002=&MultiplyLinearColor(Linear Color=<rgba 1.5 2 0 1>, Scale Factor=&MakeVector4_FromFloat(Float=0.7))) {isSetParameters=true}
   &DynamicMaterialParameters(
    Index 0 Param 1=&RandomRangeFloat(Minimum=1.5, Maximum=2)
    Index 0 Param 2=&RandomRangeFloat(Minimum=0.5, Maximum=0.7)
    Index 0 Param 3=&RandomRangeFloat(Minimum=0.02, Maximum=-0.02)
    Index 0 Param 4=&RandomRangeFloat(Minimum=-0.3, Maximum=-0.7)
    Index 2 Param 1=&RandomRangeFloat(Minimum=2, Maximum=4)
    Index 2 Param 2=&RandomRangeFloat(Minimum=1, Maximum=1.5)
    Index 2 Param 3=&RandomRangeFloat(Minimum=-0.01, Maximum=0.01)
    Index 2 Param 4=&RandomRangeFloat(Minimum=-0.8, Maximum=-1)
    Index 3 Param 1=&Multiply_Float(
     A=&MakeCustomFloatFromBool(Bool=$User.bIsFrontEndPreview, True Float=0.2, False Float=0.28)
     B=&FN_PickFloatDependOnPlatform(TargetPlatformActive=1, TargetPlatformNoActive=0.8)
    )
    Index 3 Param 2=&Multiply_Float(
     A=&MakeCustomFloatFromBool(Bool=$User.bIsFrontEndPreview, True Float=6, False Float=9)
     B=&FN_PickFloatDependOnPlatform(TargetPlatformActive=1, TargetPlatformNoActive=0.8)
    )
    Module.Index0Param0Write=true
    Module.Index0Param1Write=true
    Module.Index0Param2Write=true
    Module.Index0Param3Write=true
    Module.Index2Param0Write=true
    Module.Index2Param1Write=true
    Module.Index2Param2Write=true
    Module.Index2Param3Write=true
    Module.Index3Param0Write=true
    Module.Index3Param1Write=true
    Module.Index3Param2Write=true
    Module.Index3Param3Write=true
    Write Parameter Index 2=true
    Write Parameter Index 3=true
   )
  ]
  particleUpdateScript [
   &ParticleState(Kill Particles When Lifetime Has Elapsed=false, Let Infinitely Lived Particles Die When Emitter Deactivates=true)
   &ScaleMeshSize(Scale Factor=&SelectVectorFromArray(
     Direct Array Index=$Particles.PartitionParticles.Partition
     Vector Selection Array=DataInterface() {propertyValues {FloatData [<1 1 1>, <1 1 1>, <1 1 1>, <1 1 1>, <0.8 0.8 0.5>, <0.7 0.7 0.4>, <0.6 0.6 0.3>]}}
    ))
   &SkeletalMeshLocation(
    Bone Sampling Mode=::NewEnumerator4
    MaskSpawnGroup=false
    Skeletal Mesh=DataInterface() {propertyValues {
      DefaultMesh=null
      ExcludeBoneName=root
      OptionalComponentClass=null
      PreviewMesh=&F_MED_CoatCheck_Pack
      SkinningMode=SkinOnTheFly
      SoftSourceActor=null
      SourceMode=Default
      UvSetIndex=0
      WholeMeshLOD=-1
      bExcludeBone=false
      bReadDeformedGeometry=true
      bRequireCurrentFrameData=true
      ComponentTags []
      FilteredBones [dyn_wing01_l]
      FilteredSockets []
      MeshUserParameter {parameter {name=None, typeDefHandle {registeredTypeIndex=1}, varData []}}
      SamplingRegions []
     }}
   )
  ]
  rendererClasses [&NiagaraMeshRendererProperties]
  renderers [
   {rendererClass=&NiagaraMeshRendererProperties, rendererIndex=0}
   {rendererClass=&NiagaraMeshRendererProperties, rendererIndex=1}
  ]
 }
 Wing_L {
  emitterUpdateScript [
   &EmitterState(Life Cycle Mode=::NewEnumerator1, Loop Behavior=::NewEnumerator1, Loop Duration Mode=::NewEnumerator1, UseLoopDelay=false)
   &SpawnBurst_Instantaneous(Spawn Count=6, Use Spawn Probability=false, UseLoopCountLimit=false)
  ]
  particleSpawnScript [
   &PartitionParticles(Number Of Partitions=7)
   &InitializeParticle(
    Color=<rgba 1 0.52 0.1 2>
    Lifetime=5
    Mesh Scale Mode=::NewEnumerator3
    Mesh Uniform Scale=3
    UsePositionOffset=false
    Write Material Random=true
   )
   &InitialMeshOrientation(
    Mesh Orientation Mode=::NewEnumerator2
    Orientation Vector=<0 0 1>
    Rotation=&SelectVectorFromArray(
     Direct Array Index=$Particles.PartitionParticles.Partition
     Vector Selection Array=DataInterface() {propertyValues {FloatData [<0 0 0.18>, <0 0 0.17>, <0 0 0.18>, <0 0 0.17>, <0 0 0.2>, <0 0 0.2>, <0 0 0.16>]}}
    )
    Use Rotation Vector=true
   )
   &InitialMeshOrientation(
    Mesh Orientation Mode=::NewEnumerator2
    Rotation=&SelectVectorFromArray(
     Direct Array Index=$Particles.PartitionParticles.Partition
     Vector Selection Array=DataInterface() {propertyValues {FloatData [<0.06 0.5 0>, <0.04 0.5 0>, <0.04 0.5 0>, <0.04 0.5 0>, <0.04 0.5 0>, <0.09 0.5 0>, <0.04 0.5 0>]}}
    )
    Use Rotation Vector=true
   ) {module=InitialMeshOrientation001}
   &DynamicMaterialParameters(
    Index 0 Param 1=&RandomRangeFloat(Minimum=0.8, Maximum=1.2)
    Index 0 Param 2=&RandomRangeFloat(Minimum=0.5, Maximum=0.7)
    Index 0 Param 3=&RandomRangeFloat(Minimum=0.02, Maximum=-0.02)
    Index 0 Param 4=&RandomRangeFloat(Minimum=-0.2, Maximum=-1)
    Index 2 Param 1=&RandomRangeFloat(Minimum=1, Maximum=2)
    Index 2 Param 2=&RandomRangeFloat(Minimum=0.4, Maximum=0.7)
    Index 2 Param 3=&RandomRangeFloat(Minimum=-0.01, Maximum=0.01)
    Index 2 Param 4=&RandomRangeFloat(Minimum=-0.5, Maximum=-0.7)
    Index 3 Param 1=&Multiply_Float(
     A=&MakeCustomFloatFromBool(Bool=$User.bIsFrontEndPreview, True Float=0.2, False Float=0.28)
     B=&FN_PickFloatDependOnPlatform(TargetPlatformActive=1, TargetPlatformNoActive=0.8)
    )
    Index 3 Param 2=&Multiply_Float(
     A=&MakeCustomFloatFromBool(Bool=$User.bIsFrontEndPreview, True Float=6, False Float=9)
     B=&FN_PickFloatDependOnPlatform(TargetPlatformActive=1, TargetPlatformNoActive=0.8)
    )
    Module.Index0Param0Write=true
    Module.Index0Param1Write=true
    Module.Index0Param2Write=true
    Module.Index0Param3Write=true
    Module.Index2Param0Write=true
    Module.Index2Param1Write=true
    Module.Index2Param2Write=true
    Module.Index2Param3Write=true
    Module.Index3Param0Write=true
    Module.Index3Param1Write=true
    Module.Index3Param2Write=true
    Module.Index3Param3Write=true
    Write Parameter Index 2=true
    Write Parameter Index 3=true
   )
  ]
  particleUpdateScript [
   &ParticleState(Kill Particles When Lifetime Has Elapsed=false, Let Infinitely Lived Particles Die When Emitter Deactivates=true)
   &ScaleMeshSize(Scale Factor=&SelectVectorFromArray(
     Direct Array Index=$Particles.PartitionParticles.Partition
     Vector Selection Array=DataInterface() {propertyValues {FloatData [<1.2 1.2 0.85>, <1.4 1.4 0.95>, <1.1 1.1 0.95>, <0.9 0.9 0.77>, <0.7 0.7 0.65>, <0.7 0.7 0.42>]}}
    ))
   &SkeletalMeshLocation(
    Bone / Socket Index=$Particles.PartitionParticles.Partition
    Bone Sampling Mode=::NewEnumerator4
    MaskSpawnGroup=false
    Sampled Position Offset=&SelectVectorFromArray(
     Direct Array Index=$Particles.PartitionParticles.Partition
     Vector Selection Array=DataInterface() {propertyValues {FloatData [<3 -1 0>, <0 0 -2>, <0 1 -1>, <0 2 -1>, <5 2 -1>, <-1 -5 -0.5>]}}
    )
    Skeletal Mesh=DataInterface() {propertyValues {
      DefaultMesh=null
      ExcludeBoneName=root
      OptionalComponentClass=null
      PreviewMesh=&F_MED_CoatCheck_Pack
      SkinningMode=SkinOnTheFly
      SoftSourceActor=null
      SourceMode=Default
      UvSetIndex=0
      WholeMeshLOD=-1
      bExcludeBone=false
      bReadDeformedGeometry=true
      bRequireCurrentFrameData=true
      ComponentTags []
      FilteredBones [feather09_upper_r, feather08_upper_r, feather07_upper_r, feather06_upper_r, feather05_upper_r, feather04_upper_r, feather02_upper_r]
      FilteredSockets []
      MeshUserParameter {parameter {name=None, typeDefHandle {registeredTypeIndex=1}, varData []}}
      SamplingRegions []
     }}
   )
   /SetVariables(
    Particles.MeshScale02:Vector3f=&Multiply_VectorByFloat(Vector=$Particles.Scale, Float=1.05)
    Particles.Position02:NiagaraPosition=&AddVectorToPosition(Position=$Particles.Position, Vector=<1 0 -3>)
   ) {isSetParameters=true}
  ]
  rendererClasses [&NiagaraMeshRendererProperties]
  renderers [
   {rendererClass=&NiagaraMeshRendererProperties, rendererIndex=0}
   {rendererClass=&NiagaraMeshRendererProperties, rendererIndex=1}
  ]
 }
 Wing_R {
  emitterUpdateScript [
   &EmitterState(Life Cycle Mode=::NewEnumerator1, Loop Behavior=::NewEnumerator1, Loop Duration Mode=::NewEnumerator1, UseLoopDelay=false)
   &SpawnBurst_Instantaneous(Spawn Count=6, Use Spawn Probability=false, UseLoopCountLimit=false)
  ]
  particleSpawnScript [
   &PartitionParticles(Number Of Partitions=7)
   &InitializeParticle(
    Color=<rgba 1 0.52 0.1 2>
    Lifetime=5
    Mesh Scale Mode=::NewEnumerator3
    Mesh Uniform Scale=3
    UsePositionOffset=false
    Write Material Random=true
   )
   &InitialMeshOrientation(
    Mesh Orientation Mode=::NewEnumerator2
    Orientation Vector=<0 0 1>
    Rotation=&SelectVectorFromArray(
     Direct Array Index=$Particles.PartitionParticles.Partition
     Vector Selection Array=DataInterface() {propertyValues {FloatData [<0 0 0.25>, <0 0 0.25>, <0 0 0.25>, <0 0 0.25>, <0 0 0.25>, <0 0 0.25>, <0 0 0.25>]}}
    )
    Use Rotation Vector=true
   )
   &InitialMeshOrientation(
    Mesh Orientation Mode=::NewEnumerator2
    Rotation=&SelectVectorFromArray(
     Direct Array Index=$Particles.PartitionParticles.Partition
     Vector Selection Array=DataInterface() {propertyValues {FloatData [<-0.01 0 -0.04>, <-0.06 0 -0.03>, <-0.08 0 -0.03>, <-0.05 0 -0.03>, <-0.05 0 -0.03>, <-0.08 0 -0.04>, <0 0 -0.03>]}}
    )
    Use Rotation Vector=true
   ) {module=InitialMeshOrientation001}
   &DynamicMaterialParameters(
    Index 0 Param 1=&RandomRangeFloat(Minimum=0.8, Maximum=1.2)
    Index 0 Param 2=&RandomRangeFloat(Minimum=0.5, Maximum=0.7)
    Index 0 Param 3=&RandomRangeFloat(Minimum=0.02, Maximum=-0.02)
    Index 0 Param 4=&RandomRangeFloat(Minimum=-0.2, Maximum=-1)
    Index 2 Param 1=&RandomRangeFloat(Minimum=1, Maximum=2)
    Index 2 Param 2=&RandomRangeFloat(Minimum=0.4, Maximum=0.7)
    Index 2 Param 3=&RandomRangeFloat(Minimum=-0.01, Maximum=0.01)
    Index 2 Param 4=&RandomRangeFloat(Minimum=-0.5, Maximum=-0.7)
    Index 3 Param 1=&Multiply_Float(
     A=&MakeCustomFloatFromBool(Bool=$User.bIsFrontEndPreview, True Float=0.2, False Float=0.28)
     B=&FN_PickFloatDependOnPlatform(TargetPlatformActive=1, TargetPlatformNoActive=0.8)
    )
    Index 3 Param 2=&Multiply_Float(
     A=&MakeCustomFloatFromBool(Bool=$User.bIsFrontEndPreview, True Float=5, False Float=9)
     B=&FN_PickFloatDependOnPlatform(TargetPlatformActive=1, TargetPlatformNoActive=0.8)
    )
    Module.Index0Param0Write=true
    Module.Index0Param1Write=true
    Module.Index0Param2Write=true
    Module.Index0Param3Write=true
    Module.Index2Param0Write=true
    Module.Index2Param1Write=true
    Module.Index2Param2Write=true
    Module.Index2Param3Write=true
    Module.Index3Param0Write=true
    Module.Index3Param1Write=true
    Module.Index3Param2Write=true
    Module.Index3Param3Write=true
    Write Parameter Index 2=true
    Write Parameter Index 3=true
   )
  ]
  particleUpdateScript [
   &ParticleState(Kill Particles When Lifetime Has Elapsed=false, Let Infinitely Lived Particles Die When Emitter Deactivates=true)
   &ScaleMeshSize(Scale Factor=&SelectVectorFromArray(
     Direct Array Index=$Particles.PartitionParticles.Partition
     Vector Selection Array=DataInterface() {propertyValues {FloatData [<-1.2 1.2 0.85>, <-1.4 1.4 0.95>, <-1.1 1.1 0.95>, <-0.9 0.9 0.77>, <-0.7 0.7 0.65>, <-0.7 0.7 0.42>]}}
    ))
   &SkeletalMeshLocation(
    Bone / Socket Index=$Particles.PartitionParticles.Partition
    Bone Sampling Mode=::NewEnumerator4
    MaskSpawnGroup=false
    Sampled Position Offset=&SelectVectorFromArray(
     Direct Array Index=$Particles.PartitionParticles.Partition
     Vector Selection Array=DataInterface() {propertyValues {FloatData [<-3 1 0>, <3 0 0>, <0 0 -2>, <0 -2 -1>, <-5 0 0>, <-1 3 -1>]}}
    )
    Skeletal Mesh=DataInterface() {propertyValues {
      DefaultMesh=null
      ExcludeBoneName=root
      OptionalComponentClass=null
      PreviewMesh=&F_MED_CoatCheck_Pack
      SkinningMode=SkinOnTheFly
      SoftSourceActor=null
      SourceMode=Default
      UvSetIndex=0
      WholeMeshLOD=-1
      bExcludeBone=false
      bReadDeformedGeometry=true
      bRequireCurrentFrameData=true
      ComponentTags []
      FilteredBones [feather09_upper_l, feather08_upper_l, feather07_upper_l, feather06_upper_l, feather05_upper_l, feather04_upper_l, feather02_upper_l]
      FilteredSockets []
      MeshUserParameter {parameter {name=None, typeDefHandle {registeredTypeIndex=1}, varData []}}
      SamplingRegions []
     }}
   )
   /SetVariables(
    Particles.MeshScale02:Vector3f=&Multiply_VectorByFloat(Vector=$Particles.Scale, Float=1.05)
    Particles.Position02:NiagaraPosition=&AddVectorToPosition(Position=$Particles.Position, Vector=<-1 0 -3>)
   ) {isSetParameters=true}
  ]
  rendererClasses [&NiagaraMeshRendererProperties]
  renderers [
   {rendererClass=&NiagaraMeshRendererProperties, rendererIndex=0}
   {rendererClass=&NiagaraMeshRendererProperties, rendererIndex=1}
  ]
 }
}
summary {
 systemName=NS_Backpack_GingerPesto_Idle
 emitters [
  {emitterName=Wing_R, rendererClasses [&NiagaraMeshRendererProperties]}
  {emitterName=BlueGlow_R, rendererClasses [&NiagaraMeshRendererProperties]}
  {emitterName=Wing_L, rendererClasses [&NiagaraMeshRendererProperties]}
  {emitterName=BlueGlow_L, rendererClasses [&NiagaraMeshRendererProperties]}
 ]
 userVariables [
  User.bIsFrontEnd {type=NiagaraBool}
  User.bIsFrontEndPreview {type=NiagaraBool}
 ]
}
