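# Control Rig for the winged CoatCheck backpack cosmetic, posing its wings and feathers from player movement states.
# Spring- or interpolation-smoothed inputs such as skydive, slide, grind, sprint, crouch, swim and strafe weight layered additive wing poses.
# RotationConstraintAttach then cascades rotation from feather05 down to feather01 in each row.
# A spring-smoothed flip-flop with staggered time delays adds a mirrored feather ripple, followed by state-weighted feather offsets.
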ControlRigBlueprint /BRCosmetics/Accessories/FORT_Backpacks/Backpack_F_MED_CoatCheck/Meshes/F_MED_CoatCheck_Pack_CtrlRig
graphs [
 New Function_ContainedGraph {nodes [
   n1 = RigVMFunctionEntryNode()
   DISPATCH_RigDispatch_ToString = RigVMDispatchNode(Value=@n1.Value) {template="DISPATCH_RigDispatch_ToString(in Value,out Result)"}
   Chop = /Script/RigVM.RigVMFunction_StringTruncate(Count=9, FromEnd=false, Name=@DISPATCH_RigDispatch_ToString.Result) {is=struct RigVMUnitNode, template="Chop::Execute(in Name,in Count,in FromEnd,out Remainder,out Chopped)"}
   n2 = &RigVMFunction_StringTruncate(Count=1, FromEnd=true, Name=@Chop.Chopped) {template="Chop::Execute(in Name,in Count,in FromEnd,out Remainder,out Chopped)"}
   DISPATCH_RigDispatch_FromString = RigVMDispatchNode(String=@n2.Chopped) {template="DISPATCH_RigDispatch_FromString(in String,out Result)"}
   Add = /Script/RigVM.RigVMFunction_MathIntAdd(B=1, A=@DISPATCH_RigDispatch_FromString.Result) {is=struct RigVMUnitNode, template="Add::Execute(in A,in B,out Result)"}
   Concat = /Script/RigVM.RigVMFunction_StringConcat(B=RigVMDispatchNode(Value=@Add.Result) {template="DISPATCH_RigDispatch_ToString(in Value,out Result)"}.Result, A=@n2.Remainder) {is=struct RigVMUnitNode, template="Concat::Execute(in A,in B,out Result)"}
   > RigVMFunctionReturnNode(Result=RigVMDispatchNode(String=&RigVMFunction_StringConcat(A=@Concat.Result, B=@Chop.Remainder) {template="Concat::Execute(in A,in B,out Result)"}.Result) {template="DISPATCH_RigDispatch_FromString(in String,out Result)"}.Result)
  ]}
 New Function_ContainedGraph {nodes [
   n1 = RigVMFunctionEntryNode()
   DISPATCH_RigDispatch_ToString = RigVMDispatchNode(Value=@n1.Value) {template="DISPATCH_RigDispatch_ToString(in Value,out Result)"}
   Chop = &RigVMFunction_StringTruncate(Count=1, FromEnd=True, Name=@DISPATCH_RigDispatch_ToString.Result) {template="Chop::Execute(in Name,in Count,in FromEnd,out Remainder,out Chopped)"}
   Concat = &RigVMFunction_StringConcat(B=r, A=@Chop.Remainder) {template="Concat::Execute(in A,in B,out Result)"}
   > RigVMFunctionReturnNode(Result=RigVMDispatchNode(String=@Concat.Result) {template="DISPATCH_RigDispatch_FromString(in String,out Result)"}.Result)
  ]}
 New Function_ContainedGraph {nodes [
   n1 = RigVMFunctionEntryNode()
   > n2 = RigVMDispatchNode(Array=@n1.InputBones) {template="DISPATCH_RigVMDispatch_ArrayIterator(in Array,out Element,out Index,out Count,out Ratio)"}
   DISPATCH_RigVMDispatch_ArrayGetNum = RigVMDispatchNode(Array=@n1.InputBones) {template="DISPATCH_RigVMDispatch_ArrayGetNum(in Array,out Num)"}
   Subtract = /Script/RigVM.RigVMFunction_MathIntSub(B=1, A=@DISPATCH_RigVMDispatch_ArrayGetNum.Num) {is=struct RigVMUnitNode, template="Subtract::Execute(in A,in B,out Result)"}
   Less = /Script/RigVM.RigVMFunction_MathIntLess(A=@n2.Index, B=@Subtract.Result) {is=struct RigVMUnitNode, template="Less::Execute(in A,in B,out Result)"}
   RigVMFunction_ControlFlowBranch = &RigVMFunction_ControlFlowBranch(ExecuteContext=@n2.ExecuteContext, Condition=@Less.Result)
   Cast = &RigVMFunction_MathBoolToInteger(Value=@Less.Result)
   Add = &RigVMFunction_MathIntAdd(B=@Cast.Result, A=@n2.Index) {template="Add::Execute(in A,in B,out Result)"}
   DISPATCH_RigVMDispatch_ArrayGetAtIndex = RigVMDispatchNode(Index=@Add.Result, Array=@n1.InputBones) {template="DISPATCH_RigVMDispatch_ArrayGetAtIndex(in Array,in Index,out Element)"}
   &RigUnit_RotationConstraintLocalSpaceOffset(
    AdvancedSettings=(InterpolationType=Average,RotationOrderForFilter=XYZ)
    Filter=(bX=true,bY=true,bZ=true)
    Parents=((Item=(Type=Bone,Name="None"),Weight=1),(Item=(Type=Bone,Name="None"),Weight=1))
    Weight=1
    bMaintainOffset=true
    Child=@DISPATCH_RigVMDispatch_ArrayGetAtIndex.Element
    Parents.1.Item=@DISPATCH_RigVMDispatch_ArrayGetAtIndex.Element
    ExecutePin=@RigVMFunction_ControlFlowBranch.True
    Parents.1.Weight=@n1.SelfInfluence
    Parents.0.Weight=@n1.ChildInfluence
    Parents.0.Item=@n2.Element
   )
   RigVMFunctionReturnNode(ExecuteContext=@n2.Completed)
  ]}
 RigVMFunctionLibrary {nodes [
   InvertSide = RigVMCollapseNode(Value=None)
   IterateString = RigVMCollapseNode(Value=None)
   RotationConstraintAttach = RigVMCollapseNode(ChildInfluence=1, ExecuteContext=(), InputBones=(), SelfInfluence=1)
  ]}
 RigVMModel {nodes [
   RigUnit_BeginExecution = &RigUnit_BeginExecution()
   n1 = &RigVMFunction_MathDoubleAdd(
    A=&RigVMFunction_MathDoubleAdd(A=RigVMVariableNode() {variable=SkydiveDiveAlpha}.Value, B=RigVMVariableNode() {variable=SkydiveFloatingAlpha}.Value).Result
    B=RigVMVariableNode() {variable=SkydiveEnterAlpha}.Value
   )
   n2 = &RigVMFunction_AlphaInterp(
    InterpSpeedDecreasing=5
    InterpSpeedIncreasing=5
    Scale=1
    bClampResult=False
    bInterpResult=True
    Value=&RigVMFunction_MathDoubleMul(
     B=&RigVMFunction_MathDoubleRemap(SourceMaximum=1, SourceMinimum=0, TargetMaximum=0, TargetMinimum=1, Value=@n1.Result).Result
     A=RigVMVariableNode() {variable=JumpOrFalling}.Value
    ).Result
   )
   &RigUnit_ModifyTransforms(
    ItemToModify="((Item=(Type=Bone,Name=\"dyn_wing01_l\"),Transform=(Rotation=(X=0,Y=0,Z=0.2164,W=0.9763),Translation=(X=0,Y=0,Z=0),Scale3D=(X=1,Y=1,Z=1))),(Item=(Type=Bone,Name=\"dyn_wing01_r\"),Transform=(Rotation=(X=0,Y=0,Z=-0.2164,W=0.9763),Translation=(X=0,Y=0,Z=0),Scale3 ...(truncated)"
    Mode=AdditiveGlobal
    WeightMinimum=0
    ExecutePin=@RigUnit_BeginExecution.ExecutePin
    Weight=@n2.Result
   )
   Absolute = &RigVMFunction_MathDoubleAbs(Value=RigVMVariableNode() {variable=ForwardBackwardMovement}.Value)
   n3 = RigVMVariableNode() {variable=LeftRightSideMovement}
   n4 = &RigVMFunction_MathDoubleAdd(
    A=&RigVMFunction_MathDoubleAdd(A=@Absolute.Result, B=&RigVMFunction_MathDoubleAbs(Value=@n3.Value).Result).Result
    B=RigVMVariableNode() {variable=isTargeting}.Value
   )
   n5 = &RigVMFunction_AlphaInterp(
    InterpSpeedDecreasing=7.5
    InterpSpeedIncreasing=15
    Scale=1
    bClampResult=False
    bInterpResult=True
    Value=&RigVMFunction_MathDoubleMul(
     B=&RigVMFunction_MathDoubleRemap(SourceMaximum=1, SourceMinimum=0, TargetMaximum=0, TargetMinimum=1, Value=@n4.Result).Result
     A=RigVMVariableNode() {variable=HeadRotation}.Value
    ).Result
   )
   > &RigUnit_ModifyTransforms(
    ItemToModify="((Item=(Type=Bone,Name=\"dyn_wing01_l\"),Transform=(Rotation=(X=0,Y=0,Z=0.3827,W=0.9239),Translation=(X=0,Y=0,Z=0),Scale3D=(X=1,Y=1,Z=1))),(Item=(Type=Bone,Name=\"dyn_wing01_r\"),Transform=(Rotation=(X=0,Y=0,Z=0,W=1),Translation=(X=0,Y=0,Z=0),Scale3D=(X=1,Y=1, ...(truncated)"
    Mode=AdditiveGlobal
    WeightMinimum=0
    Weight=@n5.Result
   )
   RigVMFunction_AlphaInterp = &RigVMFunction_AlphaInterp(
    InterpSpeedDecreasing=200
    InterpSpeedIncreasing=200
    Scale=1
    bClampResult=False
    bInterpResult=True
    Value=RigVMVariableNode() {variable=SpineSwimmingAlpha}.Value
   )
   > &RigUnit_ModifyTransforms(
    ItemToModify="((Item=(Type=Bone,Name=\"dyn_wing01_l\"),Transform=(Rotation=(X=-0.1565,Y=0.1472,Z=0.1919,W=0.9576),Translation=(X=0,Y=0,Z=0),Scale3D=(X=1,Y=1,Z=1))),(Item=(Type=Bone,Name=\"dyn_wing01_r\"),Transform=(Rotation=(X=-0.1565,Y=-0.1472,Z=-0.1919,W=0.9576),Translati ...(truncated)"
    Mode=AdditiveGlobal
    WeightMinimum=0
    Weight=@RigVMFunction_AlphaInterp.Result
   )
   n6 = /Script/ControlRig.RigUnit_SpringInterpV2(
    CriticalDamping=1.5
    Current=0
    Force=0
    Strength=2
    TargetVelocityAmount=0
    bInitializeFromTarget=true
    bUseCurrentInput=False
    Target=RigVMVariableNode() {variable=SkydiveFloatingAlpha}.Value
   ) {
    is=struct RigVMUnitNode
    template="SpringInterp::Execute(in Target,in Strength,in CriticalDamping,in Force,in bUseCurrentInput,in Current,in TargetVelocityAmount,in bInitializeFromTarget,out Result,out Velocity)"
   }
   > &RigUnit_ModifyTransforms(
    ItemToModify="((Item=(Type=Bone,Name=\"def_wing_main_l\"),Transform=(Rotation=(X=0.0864,Y=0.0114,Z=0.13,W=0.9877),Translation=(X=1,Y=-1,Z=0),Scale3D=(X=1,Y=1,Z=1))),(Item=(Type=Bone,Name=\"def_wing_main_r\"),Transform=(Rotation=(X=0.0864,Y=-0.0114,Z=-0.13,W=0.9877),Translat ...(truncated)"
    Mode=AdditiveGlobal
    WeightMinimum=0
    Weight=@n6.Result
   )
   n7 = &RigUnit_SpringInterpV2(
    CriticalDamping=0.9
    Current=0
    Force=0
    Strength=2
    TargetVelocityAmount=0
    bInitializeFromTarget=true
    bUseCurrentInput=False
    Target=RigVMVariableNode() {variable=SlidingAlpha}.Value
   ) {template="SpringInterp::Execute(in Target,in Strength,in CriticalDamping,in Force,in bUseCurrentInput,in Current,in TargetVelocityAmount,in bInitializeFromTarget,out Result,out Velocity)"}
   > &RigUnit_ModifyTransforms(
    ItemToModify="((Item=(Type=Bone,Name=\"dyn_wing01_l\"),Transform=(Rotation=(X=0.0426,Y=0.0094,Z=0.2162,W=0.9754),Translation=(X=0,Y=0,Z=0),Scale3D=(X=1,Y=1,Z=1))),(Item=(Type=Bone,Name=\"dyn_wing01_r\"),Transform=(Rotation=(X=0.041,Y=-0.0149,Z=-0.3417,W=0.9388),Translation= ...(truncated)"
    Mode=AdditiveGlobal
    WeightMinimum=0
    Weight=@n7.Result
   )
   Subtract = &RigVMFunction_MathDoubleSub(B=RigVMVariableNode() {variable=GrindingSprintAlpha}.Value, A=RigVMVariableNode() {variable=GrindingAlpha}.Value)
   n8 = &RigUnit_SpringInterpV2(
    CriticalDamping=2
    Current=0
    Force=0
    Strength=1
    TargetVelocityAmount=0
    bInitializeFromTarget=true
    bUseCurrentInput=False
    Target=@Subtract.Result
   ) {template="SpringInterp::Execute(in Target,in Strength,in CriticalDamping,in Force,in bUseCurrentInput,in Current,in TargetVelocityAmount,in bInitializeFromTarget,out Result,out Velocity)"}
   > &RigUnit_ModifyTransforms(
    ItemToModify="((Item=(Type=Bone,Name=\"dyn_wing01_l\"),Transform=(Rotation=(X=0.2179,Y=0.0331,Z=-0.052,W=0.974),Translation=(X=0,Y=0,Z=0),Scale3D=(X=1,Y=1,Z=1))),(Item=(Type=Bone,Name=\"dyn_wing01_r\"),Transform=(Rotation=(X=0.1976,Y=0.1212,Z=0.1503,W=0.9611),Translation=(X ...(truncated)"
    Mode=AdditiveGlobal
    WeightMinimum=0
    Weight=@n8.Result
   )
   Multiply = &RigVMFunction_MathDoubleMul(B=RigVMVariableNode() {variable=GrindingSprintAlpha}.Value, A=RigVMVariableNode() {variable=GrindingAlpha}.Value)
   n9 = &RigUnit_SpringInterpV2(
    CriticalDamping=2
    Current=0
    Force=0
    Strength=1
    TargetVelocityAmount=0
    bInitializeFromTarget=true
    bUseCurrentInput=False
    Target=@Multiply.Result
   ) {template="SpringInterp::Execute(in Target,in Strength,in CriticalDamping,in Force,in bUseCurrentInput,in Current,in TargetVelocityAmount,in bInitializeFromTarget,out Result,out Velocity)"}
   > &RigUnit_ModifyTransforms(
    ItemToModify="((Item=(Type=Bone,Name=\"dyn_wing01_l\"),Transform=(Rotation=(X=-0.0057,Y=-0.1304,Z=-0.0432,W=0.9905),Translation=(X=0,Y=0,Z=0),Scale3D=(X=1,Y=1,Z=1))),(Item=(Type=Bone,Name=\"dyn_wing01_r\"),Transform=(Rotation=(X=-0.0057,Y=0.1304,Z=0.0432,W=0.9905),Translati ...(truncated)"
    Mode=AdditiveGlobal
    WeightMinimum=0
    Weight=@n9.Result
   )
   n10 = &RigUnit_SpringInterpV2(
    CriticalDamping=0.9
    Current=0
    Force=0
    Strength=2
    TargetVelocityAmount=0
    bInitializeFromTarget=true
    bUseCurrentInput=False
    Target=RigVMVariableNode() {variable=TacticalSprintingAlpha}.Value
   ) {template="SpringInterp::Execute(in Target,in Strength,in CriticalDamping,in Force,in bUseCurrentInput,in Current,in TargetVelocityAmount,in bInitializeFromTarget,out Result,out Velocity)"}
   > &RigUnit_ModifyTransforms(
    ItemToModify="((Item=(Type=Bone,Name=\"def_wing_main_l\"),Transform=(Rotation=(X=0.1996,Y=-0.0796,Z=0.2746,W=0.9372),Translation=(X=-2,Y=0,Z=0),Scale3D=(X=1,Y=1,Z=1))),(Item=(Type=Bone,Name=\"def_wing_main_r\"),Transform=(Rotation=(X=0.2028,Y=0.0708,Z=-0.3152,W=0.9244),Tran ...(truncated)"
    Mode=AdditiveGlobal
    WeightMinimum=0
    Weight=@n10.Result
   )
   Add = &RigVMFunction_MathDoubleAdd(A=RigVMVariableNode() {variable=SkydiveDiveAlpha}.Value, B=RigVMVariableNode() {variable=SkydiveEnterAlpha}.Value)
   Clamp = &RigVMFunction_MathDoubleClamp(Value=@Add.Result)
   n11 = &RigUnit_SpringInterpV2(
    CriticalDamping=0.85
    Current=0
    Force=0
    Strength=2
    TargetVelocityAmount=0
    bInitializeFromTarget=true
    bUseCurrentInput=False
    Target=@Clamp.Result
   ) {template="SpringInterp::Execute(in Target,in Strength,in CriticalDamping,in Force,in bUseCurrentInput,in Current,in TargetVelocityAmount,in bInitializeFromTarget,out Result,out Velocity)"}
   > &RigUnit_ModifyTransforms(
    ItemToModify="((Item=(Type=Bone,Name=\"def_wing_main_l\"),Transform=(Rotation=(X=-0.0534,Y=0.0308,Z=0.2602,W=0.9636),Translation=(X=0,Y=0,Z=0),Scale3D=(X=1,Y=1,Z=1))),(Item=(Type=Bone,Name=\"def_wing_main_r\"),Transform=(Rotation=(X=-0.0534,Y=-0.0308,Z=-0.2602,W=0.9636),Tra ...(truncated)"
    Mode=AdditiveGlobal
    WeightMinimum=0
    Weight=@n11.Result
   )
   n12 = &RigUnit_SpringInterpV2(
    CriticalDamping=0.9
    Current=0
    Force=0
    Strength=2
    TargetVelocityAmount=0
    bInitializeFromTarget=true
    bUseCurrentInput=False
    Target=RigVMVariableNode() {variable=isCrouching}.Value
   ) {template="SpringInterp::Execute(in Target,in Strength,in CriticalDamping,in Force,in bUseCurrentInput,in Current,in TargetVelocityAmount,in bInitializeFromTarget,out Result,out Velocity)"}
   > &RigUnit_ModifyTransforms(
    ItemToModify="((Item=(Type=Bone,Name=\"def_wing_main_l\"),Transform=(Rotation=(X=0,Y=0,Z=0.2588,W=0.9659),Translation=(X=-1,Y=0.5,Z=0),Scale3D=(X=1,Y=1,Z=1))),(Item=(Type=Bone,Name=\"def_wing_main_r\"),Transform=(Rotation=(X=0,Y=0,Z=-0.2588,W=0.9659),Translation=(X=1,Y=0.5, ...(truncated)"
    Mode=AdditiveGlobal
    WeightMinimum=0
    Weight=@n12.Result
   )
   n13 = RigVMVariableNode() {variable=isCrouching}
   n14 = &RigUnit_SpringInterpV2(
    CriticalDamping=0.9
    Current=0
    Force=0
    Strength=2
    TargetVelocityAmount=0
    bInitializeFromTarget=true
    bUseCurrentInput=False
    Target=&RigVMFunction_MathDoubleMul(B=&RigVMFunction_MathDoubleMul(B=0.25, A=@n13.Value).Result, A=RigVMVariableNode() {variable=isTargeting}.Value).Result
   ) {template="SpringInterp::Execute(in Target,in Strength,in CriticalDamping,in Force,in bUseCurrentInput,in Current,in TargetVelocityAmount,in bInitializeFromTarget,out Result,out Velocity)"}
   > &RigUnit_ModifyTransforms(
    ItemToModify="((Item=(Type=Bone,Name=\"def_wing_main_l\"),Transform=(Rotation=(X=0,Y=0,Z=0.1736,W=0.9848),Translation=(X=0,Y=0,Z=0),Scale3D=(X=1,Y=1,Z=1))),(Item=(Type=Bone,Name=\"def_wing_main_r\"),Transform=(Rotation=(X=0,Y=0,Z=-0.2588,W=0.9659),Translation=(X=0,Y=0,Z=0), ...(truncated)"
    Mode=AdditiveGlobal
    WeightMinimum=0
    Weight=@n14.Result
   )
   n15 = RigVMVariableNode() {variable=LeftRightSideMovement}
   n16 = &RigUnit_SpringInterpV2(
    CriticalDamping=0.9
    Current=0
    Force=0
    Strength=2
    TargetVelocityAmount=0
    bInitializeFromTarget=true
    bUseCurrentInput=False
    Target=@n15.Value
   ) {template="SpringInterp::Execute(in Target,in Strength,in CriticalDamping,in Force,in bUseCurrentInput,in Current,in TargetVelocityAmount,in bInitializeFromTarget,out Result,out Velocity)"}
   > &RigUnit_ModifyTransforms(
    ItemToModify="((Item=(Type=Bone,Name=\"def_wing_main_l\"),Transform=(Rotation=(X=0.0076,Y=0.0868,Z=-0.0868,W=0.9924),Translation=(X=0,Y=0,Z=0),Scale3D=(X=1,Y=1,Z=1))),(Item=(Type=Bone,Name=\"def_wing_main_r\"),Transform=(Rotation=(X=0.0076,Y=0.0868,Z=-0.0868,W=0.9924),Trans ...(truncated)"
    Mode=AdditiveGlobal
    WeightMinimum=0
    Weight=@n16.Result
   )
   n17 = &RigUnit_SpringInterpV2(
    CriticalDamping=0.9
    Current=0
    Force=0
    Strength=2
    TargetVelocityAmount=0
    bInitializeFromTarget=true
    bUseCurrentInput=False
    Target=/Script/RigVM.RigVMFunction_MathDoubleNegate(Value=@n15.Value) {is=struct RigVMUnitNode, template="Negate::Execute(in Value,out Result)"}.Result
   ) {template="SpringInterp::Execute(in Target,in Strength,in CriticalDamping,in Force,in bUseCurrentInput,in Current,in TargetVelocityAmount,in bInitializeFromTarget,out Result,out Velocity)"}
   > &RigUnit_ModifyTransforms(
    ItemToModify="((Item=(Type=Bone,Name=\"def_wing_main_l\"),Transform=(Rotation=(X=0.0076,Y=-0.0868,Z=0.0868,W=0.9924),Translation=(X=0,Y=0,Z=0),Scale3D=(X=1,Y=1,Z=1))),(Item=(Type=Bone,Name=\"def_wing_main_r\"),Transform=(Rotation=(X=0.0076,Y=-0.0868,Z=0.0868,W=0.9924),Trans ...(truncated)"
    Mode=AdditiveGlobal
    WeightMinimum=0
    Weight=@n17.Result
   )
   n18 = &RigUnit_ItemArray(Items=((Type=Bone,Name="feather05_upper_l"),(Type=Bone,Name="feather04_upper_l"),(Type=Bone,Name="feather03_upper_l"),(Type=Bone,Name="feather02_upper_l"),(Type=Bone,Name="feather01_upper_l")))
   > RotationConstraintAttach_3 = RigVMFunctionReferenceNode(ChildInfluence=0.75, SelfInfluence=0.25, InputBones=@n18.Items)
   n19 = &RigUnit_ItemArray(Items=((Type=Bone,Name="feather05_middle_l"),(Type=Bone,Name="feather04_middle_l"),(Type=Bone,Name="feather03_middle_l"),(Type=Bone,Name="feather02_middle_l"),(Type=Bone,Name="feather01_middle_l")))
   > RotationConstraintAttach_1_3 = RigVMFunctionReferenceNode(ChildInfluence=0.5, SelfInfluence=0.5, InputBones=@n19.Items)
   n20 = &RigUnit_ItemArray(Items=((Type=Bone,Name="feather05_bottom_l"),(Type=Bone,Name="feather04_bottom_l"),(Type=Bone,Name="feather03_bottom_l"),(Type=Bone,Name="feather02_bottom_l"),(Type=Bone,Name="feather01_bottom_l")))
   > RotationConstraintAttach_1_1_2 = RigVMFunctionReferenceNode(ChildInfluence=1, SelfInfluence=0, InputBones=@n20.Items)
   n21 = &RigUnit_ItemArray(Items=((Type=Bone,Name="feather05_upper_r"),(Type=Bone,Name="feather04_upper_r"),(Type=Bone,Name="feather03_upper_r"),(Type=Bone,Name="feather02_upper_r"),(Type=Bone,Name="feather01_upper_r")))
   > RotationConstraintAttach_2_1 = RigVMFunctionReferenceNode(ChildInfluence=0.75, SelfInfluence=0.25, InputBones=@n21.Items)
   n22 = &RigUnit_ItemArray(Items=((Type=Bone,Name="feather05_middle_r"),(Type=Bone,Name="feather04_middle_r"),(Type=Bone,Name="feather03_middle_r"),(Type=Bone,Name="feather02_middle_r"),(Type=Bone,Name="feather01_middle_r")))
   > RotationConstraintAttach_1_2_1 = RigVMFunctionReferenceNode(ChildInfluence=0.5, SelfInfluence=0.5, InputBones=@n22.Items)
   n23 = &RigUnit_ItemArray(Items=((Type=Bone,Name="feather05_bottom_r"),(Type=Bone,Name="feather04_bottom_r"),(Type=Bone,Name="feather03_bottom_r"),(Type=Bone,Name="feather02_bottom_r"),(Type=Bone,Name="feather01_bottom_r")))
   > RotationConstraintAttach_1_1_1_1 = RigVMFunctionReferenceNode(ChildInfluence=1, SelfInfluence=0, InputBones=@n23.Items)
   RigVMFunction_MathBoolFlipFlop = /Script/RigVM.RigVMFunction_MathBoolFlipFlop(Duration=1.5) {is=struct RigVMUnitNode}
   RigVMFunction_MathBoolToFloat = &RigVMFunction_MathBoolToFloat(Value=@RigVMFunction_MathBoolFlipFlop.Result)
   RigVMFunction_AnimEasing = /Script/RigVM.RigVMFunction_AnimEasing(SourceMaximum=1, SourceMinimum=0, TargetMaximum=1, TargetMinimum=-0.5, Type=BounceEaseInOut, Value=@RigVMFunction_MathBoolToFloat.Result) {is=struct RigVMUnitNode}
   SpringInterp = &RigUnit_SpringInterpV2(
    CriticalDamping=0.3
    Current=0
    Force=0
    Strength=0.6
    TargetVelocityAmount=0
    bInitializeFromTarget=true
    bUseCurrentInput=false
    Target=@RigVMFunction_AnimEasing.Result
   ) {template="SpringInterp::Execute(in Target,in Strength,in CriticalDamping,in Force,in bUseCurrentInput,in Current,in TargetVelocityAmount,in bInitializeFromTarget,out Result,out Velocity)"}
   TimeOffset = /Script/RigVM.RigVMFunction_TimeOffsetFloat(BufferSize=8, SecondsAgo=0.4, TimeRange=1, Value=@SpringInterp.Result) {is=struct RigVMUnitNode, template="TimeOffset::Execute(in Value,in SecondsAgo,in BufferSize,in TimeRange,out Result)"}
   n24 = &RigVMFunction_TimeOffsetFloat(BufferSize=8, SecondsAgo=0.6, TimeRange=1, Value=@SpringInterp.Result) {template="TimeOffset::Execute(in Value,in SecondsAgo,in BufferSize,in TimeRange,out Result)"}
   n25 = &RigVMFunction_TimeOffsetFloat(BufferSize=8, SecondsAgo=0.8, TimeRange=1, Value=@SpringInterp.Result) {template="TimeOffset::Execute(in Value,in SecondsAgo,in BufferSize,in TimeRange,out Result)"}
   n26 = &RigVMFunction_TimeOffsetFloat(BufferSize=16, SecondsAgo=0.9, TimeRange=1, Value=@SpringInterp.Result) {template="TimeOffset::Execute(in Value,in SecondsAgo,in BufferSize,in TimeRange,out Result)"}
   n27 = &RigVMFunction_TimeOffsetFloat(BufferSize=16, SecondsAgo=0.2, TimeRange=1, Value=@n26.Result) {template="TimeOffset::Execute(in Value,in SecondsAgo,in BufferSize,in TimeRange,out Result)"}
   n28 = &RigVMFunction_TimeOffsetFloat(BufferSize=16, SecondsAgo=0.4, TimeRange=1, Value=@n26.Result) {template="TimeOffset::Execute(in Value,in SecondsAgo,in BufferSize,in TimeRange,out Result)"}
   n29 = &RigVMFunction_TimeOffsetFloat(BufferSize=16, SecondsAgo=0.6, TimeRange=1, Value=@n26.Result) {template="TimeOffset::Execute(in Value,in SecondsAgo,in BufferSize,in TimeRange,out Result)"}
   n30 = &RigVMFunction_TimeOffsetFloat(BufferSize=8, SecondsAgo=0.2, TimeRange=1, Value=@SpringInterp.Result) {template="TimeOffset::Execute(in Value,in SecondsAgo,in BufferSize,in TimeRange,out Result)"}
   > &RigUnit_ModifyTransforms(
    ItemToModify="((Item=(Type=Bone,Name=\"feather08_upper_l\"),Transform=(Rotation=(X=0,Y=0,Z=0,W=1),Translation=(X=0,Y=0,Z=0),Scale3D=(X=1,Y=1,Z=1))),(Item=(Type=Bone,Name=\"feather07_upper_l\"),Transform=(Rotation=(X=0,Y=0,Z=0,W=1),Translation=(X=0,Y=0,Z=0),Scale3D=(X=1,Y=1, ...(truncated)"
    Weight=1
    WeightMinimum=0
    ItemToModify.1.Transform.Translation.X=@TimeOffset.Result
    ItemToModify.2.Transform.Translation.X=@n24.Result
    ItemToModify.3.Transform.Translation.X=@n25.Result
    ItemToModify.4.Transform.Translation.X=@n26.Result
    ItemToModify.5.Transform.Translation.X=@n27.Result
    ItemToModify.6.Transform.Translation.X=@n28.Result
    ItemToModify.7.Transform.Translation.X=@n29.Result
    ItemToModify.0.Transform.Translation.X=@n30.Result
    ItemToModify.8.Transform.Translation.X=@SpringInterp.Result
   )
   Negate = &RigVMFunction_MathDoubleNegate(Value=@SpringInterp.Result) {template="Negate::Execute(in Value,out Result)"}
   > &RigUnit_ModifyTransforms(
    ItemToModify="((Item=(Type=Bone,Name=\"feather08_upper_r\"),Transform=(Rotation=(X=0,Y=0,Z=0,W=1),Translation=(X=0,Y=0,Z=0),Scale3D=(X=1,Y=1,Z=1))),(Item=(Type=Bone,Name=\"feather07_upper_r\"),Transform=(Rotation=(X=0,Y=0,Z=0,W=1),Translation=(X=0,Y=0,Z=0),Scale3D=(X=1,Y=1, ...(truncated)"
    Weight=1
    WeightMinimum=0
    ItemToModify.8.Transform.Translation.X=@Negate.Result
    ItemToModify.1.Transform.Translation.X=&RigVMFunction_MathDoubleNegate(Value=@TimeOffset.Result) {template="Negate::Execute(in Value,out Result)"}.Result
    ItemToModify.2.Transform.Translation.X=&RigVMFunction_MathDoubleNegate(Value=@n24.Result) {template="Negate::Execute(in Value,out Result)"}.Result
    ItemToModify.3.Transform.Translation.X=&RigVMFunction_MathDoubleNegate(Value=@n25.Result) {template="Negate::Execute(in Value,out Result)"}.Result
    ItemToModify.4.Transform.Translation.X=&RigVMFunction_MathDoubleNegate(Value=@n26.Result) {template="Negate::Execute(in Value,out Result)"}.Result
    ItemToModify.5.Transform.Translation.X=&RigVMFunction_MathDoubleNegate(Value=@n27.Result) {template="Negate::Execute(in Value,out Result)"}.Result
    ItemToModify.6.Transform.Translation.X=&RigVMFunction_MathDoubleNegate(Value=@n28.Result) {template="Negate::Execute(in Value,out Result)"}.Result
    ItemToModify.7.Transform.Translation.X=&RigVMFunction_MathDoubleNegate(Value=@n29.Result) {template="Negate::Execute(in Value,out Result)"}.Result
    ItemToModify.0.Transform.Translation.X=&RigVMFunction_MathDoubleNegate(Value=@n30.Result) {template="Negate::Execute(in Value,out Result)"}.Result
   )
   n31 = RigVMVariableNode() {variable=GrindingSprintAlpha}
   n32 = RigVMVariableNode() {variable=SkydiveFloatingAlpha}
   n33 = &RigUnit_SpringInterpV2(
    CriticalDamping=0.5
    Current=0
    Force=0
    Strength=0.5
    TargetVelocityAmount=0
    bInitializeFromTarget=true
    bUseCurrentInput=False
    Target=&RigVMFunction_MathDoubleClamp(Value=&RigVMFunction_MathDoubleAdd(A=@n31.Value, B=@n32.Value).Result).Result
   ) {template="SpringInterp::Execute(in Target,in Strength,in CriticalDamping,in Force,in bUseCurrentInput,in Current,in TargetVelocityAmount,in bInitializeFromTarget,out Result,out Velocity)"}
   > &RigUnit_ModifyTransforms(
    ItemToModify="((Item=(Type=Bone,Name=\"feather08_upper_l\"),Transform=(Rotation=(X=-0.0131,Y=-0.0416,Z=-0.3004,W=0.9528),Translation=(X=0,Y=-3,Z=0),Scale3D=(X=1,Y=1,Z=1))),(Item=(Type=Bone,Name=\"feather08_upper_r\"),Transform=(Rotation=(X=-0.0131,Y=-0.0416,Z=-0.3004,W=0.95 ...(truncated)"
    WeightMinimum=0
    Weight=@n33.Result
   )
   n34 = &RigUnit_SpringInterpV2(
    CriticalDamping=0.5
    Current=0
    Force=0
    Strength=0.5
    TargetVelocityAmount=0
    bInitializeFromTarget=true
    bUseCurrentInput=False
    Target=&RigVMFunction_MathDoubleMul(B=0.5, A=RigVMVariableNode() {variable=SlidingAlpha}.Value).Result
   ) {template="SpringInterp::Execute(in Target,in Strength,in CriticalDamping,in Force,in bUseCurrentInput,in Current,in TargetVelocityAmount,in bInitializeFromTarget,out Result,out Velocity)"}
   > &RigUnit_ModifyTransforms(
    ItemToModify="((Item=(Type=Bone,Name=\"feather08_upper_l\"),Transform=(Rotation=(X=-0.0131,Y=-0.0416,Z=-0.3004,W=0.9528),Translation=(X=0,Y=-3,Z=0),Scale3D=(X=1,Y=1,Z=1))),(Item=(Type=Bone,Name=\"feather08_upper_r\"),Transform=(Rotation=(X=-0.0131,Y=-0.0416,Z=-0.3004,W=0.95 ...(truncated)"
    WeightMinimum=0
    Weight=@n34.Result
   )
   n35 = &RigUnit_SpringInterpV2(
    CriticalDamping=0.9
    Current=0
    Force=0
    Strength=3
    TargetVelocityAmount=0
    bInitializeFromTarget=true
    bUseCurrentInput=False
    Target=RigVMVariableNode() {variable=JumpOrFalling}.Value
   ) {template="SpringInterp::Execute(in Target,in Strength,in CriticalDamping,in Force,in bUseCurrentInput,in Current,in TargetVelocityAmount,in bInitializeFromTarget,out Result,out Velocity)"}
   Remap = &RigVMFunction_MathDoubleRemap(SourceMaximum=1, SourceMinimum=0, TargetMaximum=0.5, TargetMinimum=0, Value=@n35.Result)
   > &RigUnit_ModifyTransforms(
    ItemToModify="((Item=(Type=Bone,Name=\"feather08_upper_l\"),Transform=(Rotation=(X=-0.0131,Y=-0.0416,Z=-0.3004,W=0.9528),Translation=(X=0,Y=-3,Z=0),Scale3D=(X=1,Y=1,Z=1))),(Item=(Type=Bone,Name=\"feather08_upper_r\"),Transform=(Rotation=(X=-0.0131,Y=-0.0416,Z=-0.3004,W=0.95 ...(truncated)"
    WeightMinimum=0
    Weight=@Remap.Result
   )
  ]}
]
hierarchy [
 root {loc=<0 0 0>, rot=<0 0 0 1>}
 pack {parent=root, loc=<0 -0.7316 0>, rot=<0 0 0 1>}
 def_wing_main_l {parent=pack, loc=<-4.6784 -1.8393 3.8605>, rot=<0.3873 -0.5529 0.7377 -0.008>}
 dyn_wing01_l {parent=def_wing_main_l, loc=<0 0 0>, rot=<0 0 0 1>}
 dyn_wing02_l {parent=dyn_wing01_l, loc=<9.2992 0 0>, rot=<0.0172 0.0052 0.0162 -0.9997>}
 dyn_wing03_l {parent=dyn_wing02_l, loc=<9.7385 0 0>, rot=<-0.0146 0.004 0.1492 -0.9887>}
 dyn_wing04_l {parent=dyn_wing03_l, loc=<9.2294 0 0>, rot=<-0.0056 -0.0024 0.2091 -0.9779>}
 dyn_wing05_l {parent=dyn_wing04_l, loc=<7.6802 0 0>, rot=<0.0088 -0.0008 0.1981 -0.9802>}
 dyn_wing_end_l {parent=dyn_wing05_l, loc=<8.7949 0 0>, rot=<0.0397 0 0 -0.9992>}
 feather08_upper_l {parent=dyn_wing04_l, loc=<-4.1539 4.3782 0.5165>, rot=<0.0427 0.3155 0.9432 0.0952>}
 feather08_middle_l {parent=feather08_upper_l, loc=<26.5293 0 0>, rot=<0.0198 0.0234 0.2005 -0.9792>}
 feather08_bottom_l {parent=feather08_middle_l, loc=<16.5145 0 0>, rot=<0.0046 0.0283 0.0302 -0.9991>}
 feather08_end_l {parent=feather08_bottom_l, loc=<17.224 0 0>, rot=<0 0 0 1>}
 feather09_upper_l {parent=dyn_wing04_l, loc=<-0.929 6.758 1.3502>, rot=<0.0703 0.3134 0.9423 0.0938>}
 feather09_middle_l {parent=feather09_upper_l, loc=<19.5834 0 0>, rot=<0.0289 0.0015 0.2252 -0.9739>}
 feather09_bottom_l {parent=feather09_middle_l, loc=<14.5606 0 0>, rot=<0.0058 0.0241 0.0373 -0.999>}
 feather09_end_l {parent=feather09_bottom_l, loc=<14.7275 0 0>, rot=<0 0 0 1>}
 feather06_upper_l {parent=dyn_wing03_l, loc=<-3.6127 2.9591 -0.6962>, rot=<0.1137 0.2944 0.9306 0.1854>}
 feather06_middle_l {parent=feather06_upper_l, loc=<20.1032 0 0>, rot=<0.0061 0.0173 0.2011 -0.9794>}
 feather06_bottom_l {parent=feather06_middle_l, loc=<14.4422 0 0>, rot=<0.0021 0.0309 0.0267 -0.9992>}
 feather06_end_l {parent=feather06_bottom_l, loc=<13.9357 0 0>, rot=<0 0 0 1>}
 feather07_upper_l {parent=dyn_wing03_l, loc=<2.3117 4.6014 -0.9161>, rot=<0.1132 0.2954 0.9175 0.2409>}
 feather07_middle_l {parent=feather07_upper_l, loc=<27.5551 0 0>, rot=<0.0136 0.0177 0.2066 -0.9782>}
 feather07_bottom_l {parent=feather07_middle_l, loc=<17.2085 0 0>, rot=<0.0039 0.039 0.0318 -0.9987>}
 feather07_end_l {parent=feather07_bottom_l, loc=<15.5122 0 0>, rot=<0 0 0 1>}
 feather04_upper_l {parent=dyn_wing02_l, loc=<-0.4321 3.3708 -0.9089>, rot=<0.1477 0.2546 0.9255 0.2385>}
 feather04_middle_l {parent=feather04_upper_l, loc=<11.0013 0 0>, rot=<-0.0066 0.0152 0.1936 -0.9809>}
 feather04_bottom_l {parent=feather04_middle_l, loc=<7.7242 0 0>, rot=<0.0004 0.0317 0.0272 -0.9991>}
 feather04_end_l {parent=feather04_bottom_l, loc=<7.6058 0 0>, rot=<0 0 0 1>}
 feather05_upper_l {parent=dyn_wing02_l, loc=<2.8983 3.9456 -0.8119>, rot=<0.1492 0.2574 0.9172 0.2651>}
 feather05_middle_l {parent=feather05_upper_l, loc=<15.1471 0 0>, rot=<-0.0033 0.0163 0.211 -0.9773>}
 feather05_bottom_l {parent=feather05_middle_l, loc=<9.8055 0 0>, rot=<0.0007 0.0332 0.0202 -0.9992>}
 feather05_end_l {parent=feather05_bottom_l, loc=<9.6878 0 0>, rot=<0 0 0 1>}
 feather03_upper_l {parent=dyn_wing02_l, loc=<-3.2693 2.9988 -0.9089>, rot=<0.1304 0.2617 0.9225 0.252>}
 feather03_middle_l {parent=feather03_upper_l, loc=<10.0747 0 0>, rot=<-0.0122 0.0139 0.1932 -0.981>}
 feather03_bottom_l {parent=feather03_middle_l, loc=<6.663 0 0>, rot=<0 0.0375 0.0022 -0.9993>}
 feather03_end_l {parent=feather03_bottom_l, loc=<6.0737 0 0>, rot=<0 0 0 1>}
 feather01_upper_l {parent=dyn_wing01_l, loc=<1.1905 2.0326 -0.748>, rot=<0.0935 0.2921 0.894 0.3266>}
 feather01_middle_l {parent=feather01_upper_l, loc=<7.3927 0 0>, rot=<-0.0211 0.0117 0.2148 -0.9764>}
 feather01_bottom_l {parent=feather01_middle_l, loc=<4.6492 0 0>, rot=<-0.003 0.0375 0.0628 -0.9973>}
 feather01_end_l {parent=feather01_bottom_l, loc=<4.1411 0 0>, rot=<0 0 0 1>}
 feather02_upper_l {parent=dyn_wing01_l, loc=<3.3712 2.6014 -0.9652>, rot=<0.1212 0.2763 0.9122 0.2773>}
 feather02_middle_l {parent=feather02_upper_l, loc=<8.135 0 0>, rot=<-0.013 0.0058 0.1708 -0.9852>}
 feather02_bottom_l {parent=feather02_middle_l, loc=<5.2799 0 0>, rot=<-0.0007 0.0337 0.0202 -0.9992>}
 feather02_end_l {parent=feather02_bottom_l, loc=<4.5278 0 0>, rot=<0 0 0 1>}
 dyn_Off_wing01_l {parent=def_wing_main_l, loc=<0 0 0>, rot=<0 0 0 1>}
 def_wing_main_r {parent=pack, loc=<4.6784 -1.8393 3.8605>, rot=<-0.008 -0.7377 -0.5529 -0.3873>}
 dyn_wing01_r {parent=def_wing_main_r, loc=<0 0 0>, rot=<0 0 0 1>}
 dyn_wing02_r {parent=dyn_wing01_r, loc=<-9.2992 0 0>, rot=<0.0172 0.0052 0.0162 -0.9997>}
 dyn_wing03_r {parent=dyn_wing02_r, loc=<-9.7385 0 0>, rot=<-0.0146 0.004 0.1492 -0.9887>}
 dyn_wing04_r {parent=dyn_wing03_r, loc=<-9.2293 0 0>, rot=<-0.0056 -0.0024 0.2091 -0.9779>}
 dyn_wing05_r {parent=dyn_wing04_r, loc=<-7.6802 0 0.0001>, rot=<0.0088 -0.0008 0.1981 -0.9802>}
 dyn_wing_end_r {parent=dyn_wing05_r, loc=<-8.7949 0 0>, rot=<0.0397 0 0 -0.9992>}
 feather08_upper_r {parent=dyn_wing04_r, loc=<4.1539 -4.3782 -0.5164>, rot=<0.0427 0.3155 0.9432 0.0952>}
 feather08_middle_r {parent=feather08_upper_r, loc=<-26.5293 -0.0001 0>, rot=<0.0198 0.0234 0.2005 -0.9792>}
 feather08_bottom_r {parent=feather08_middle_r, loc=<-16.5145 0 0>, rot=<0.0046 0.0283 0.0302 -0.9991>}
 feather08_end_r {parent=feather08_bottom_r, loc=<-17.2239 0 0>, rot=<0 0 0 1>}
 feather09_upper_r {parent=dyn_wing04_r, loc=<0.929 -6.758 -1.3502>, rot=<0.0703 0.3134 0.9423 0.0938>}
 feather09_middle_r {parent=feather09_upper_r, loc=<-19.5833 0 0>, rot=<0.0289 0.0015 0.2252 -0.9739>}
 feather09_bottom_r {parent=feather09_middle_r, loc=<-14.5606 0 0>, rot=<0.0058 0.0241 0.0373 -0.999>}
 feather09_end_r {parent=feather09_bottom_r, loc=<-14.7275 0 0>, rot=<0 0 0 1>}
 feather06_upper_r {parent=dyn_wing03_r, loc=<3.6128 -2.959 0.6962>, rot=<0.1137 0.2944 0.9306 0.1854>}
 feather06_middle_r {parent=feather06_upper_r, loc=<-20.1031 0 0>, rot=<0.0061 0.0173 0.2011 -0.9794>}
 feather06_bottom_r {parent=feather06_middle_r, loc=<-14.4422 0 -0.0001>, rot=<0.0021 0.0309 0.0267 -0.9992>}
 feather06_end_r {parent=feather06_bottom_r, loc=<-13.9357 0 0.0001>, rot=<0 0 0 1>}
 feather07_upper_r {parent=dyn_wing03_r, loc=<-2.3116 -4.6014 0.9161>, rot=<0.1132 0.2954 0.9175 0.2409>}
 feather07_middle_r {parent=feather07_upper_r, loc=<-27.5551 0 0>, rot=<0.0136 0.0177 0.2066 -0.9782>}
 feather07_bottom_r {parent=feather07_middle_r, loc=<-17.2085 0 0>, rot=<0.0039 0.039 0.0318 -0.9987>}
 feather07_end_r {parent=feather07_bottom_r, loc=<-15.5121 0.0001 0>, rot=<0 0 0 -1>}
 feather04_upper_r {parent=dyn_wing02_r, loc=<0.4321 -3.3708 0.9089>, rot=<0.1477 0.2546 0.9255 0.2385>}
 feather04_middle_r {parent=feather04_upper_r, loc=<-11.0013 0 0>, rot=<-0.0066 0.0152 0.1936 -0.9809>}
 feather04_bottom_r {parent=feather04_middle_r, loc=<-7.7242 0 0>, rot=<0.0004 0.0317 0.0272 -0.9991>}
 feather04_end_r {parent=feather04_bottom_r, loc=<-7.6057 0 0>, rot=<0 0 0 1>}
 feather05_upper_r {parent=dyn_wing02_r, loc=<-2.8983 -3.9456 0.812>, rot=<0.1492 0.2574 0.9172 0.2651>}
 feather05_middle_r {parent=feather05_upper_r, loc=<-15.1472 0 0>, rot=<-0.0033 0.0163 0.211 -0.9773>}
 feather05_bottom_r {parent=feather05_middle_r, loc=<-9.8056 0 0>, rot=<0.0007 0.0332 0.0202 -0.9992>}
 feather05_end_r {parent=feather05_bottom_r, loc=<-9.6878 -0.0001 0>, rot=<0 0 0 1>}
 feather03_upper_r {parent=dyn_wing02_r, loc=<3.2692 -2.9988 0.9089>, rot=<0.1304 0.2617 0.9225 0.252>}
 feather03_middle_r {parent=feather03_upper_r, loc=<-10.0747 0 0>, rot=<-0.0122 0.0139 0.1932 -0.981>}
 feather03_bottom_r {parent=feather03_middle_r, loc=<-6.663 0 0>, rot=<0 0.0375 0.0022 -0.9993>}
 feather03_end_r {parent=feather03_bottom_r, loc=<-6.0738 0 0>, rot=<0 0 0 1>}
 feather01_upper_r {parent=dyn_wing01_r, loc=<-1.1905 -2.0326 0.748>, rot=<0.0935 0.2921 0.894 0.3266>}
 feather01_middle_r {parent=feather01_upper_r, loc=<-7.3927 0 0>, rot=<-0.0211 0.0117 0.2148 -0.9764>}
 feather01_bottom_r {parent=feather01_middle_r, loc=<-4.6492 0 0>, rot=<-0.003 0.0375 0.0628 -0.9973>}
 feather01_end_r {parent=feather01_bottom_r, loc=<-4.1412 0 0>, rot=<0 0 0 1>}
 feather02_upper_r {parent=dyn_wing01_r, loc=<-3.3712 -2.6014 0.9652>, rot=<0.1212 0.2763 0.9122 0.2773>}
 feather02_middle_r {parent=feather02_upper_r, loc=<-8.135 0 0>, rot=<-0.013 0.0058 0.1708 -0.9852>}
 feather02_bottom_r {parent=feather02_middle_r, loc=<-5.2799 0 0>, rot=<-0.0007 0.0337 0.0202 -0.9992>}
 feather02_end_r {parent=feather02_bottom_r, loc=<-4.5278 0 0>, rot=<0 0 0 -1>}
 dyn_Off_wing01_r {parent=def_wing_main_r, loc=<0 0 0>, rot=<0 0 0 1>}
]
variables [
 ChildRotationInfluence {type=double, default=0.25}
 ForwardBackwardMovement {type=double, default=0}
 GrindingAlpha {type=double, default=0}
 GrindingSprintAlpha {type=double, default=0}
 HeadRotation {type=double, default=0}
 JumpOrFalling {type=double, default=0}
 LeftRightSideMovement {type=double, default=0}
 SelfRotationInfluence {type=double, default=0.75}
 SkydiveDiveAlpha {type=double, default=0}
 SkydiveEnterAlpha {type=double, default=0}
 SkydiveFloatingAlpha {type=double, default=0}
 SlidingAlpha {type=double, default=0}
 SpineSwimmingAlpha {type=double, default=0}
 TacticalSprintingAlpha {type=double, default=0}
 isCrouching {type=double, default=0}
 isTargeting {type=double, default=0}
]
