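# Animation Blueprint for the CoatCheck back bling, a winged pack that reacts to player movement.
# AnimGraph applies root offsets, chained AnimDynamics springs on wing and feather bones (right wing04/05 nodes are unconnected), a root shrink transform, then a Control Rig.
# EventGraph turns movement, skydive, swimming and turn-in-place states into rig alphas.
# It also sets delayed pack-hidden flags when an opened glider carries the CoatCheck tag.
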AnimBlueprint /BRCosmetics/Accessories/FORT_Backpacks/Backpack_F_MED_CoatCheck/Meshes/F_MED_CoatCheck_Pack_AnimBP
target_skeleton=&/F_MED_CoatCheck_Pack_Skeleton
graphs [
 AnimGraph {nodes [
   n1 = Anim Dynamics - Bone: dyn_wing04_r(Alpha=1, ExternalForce="0,0,0") {
    alpha=0.5
    angularDampingOverride=0.999
    angularSpringConstant=2000
    bAngularSpring=true
    bGravityOverrideInSimSpace=true
    bOverrideAngularDamping=true
    bUseGravityOverride=true
    simulationSpace=RootRelative
    alphaBoolBlend {blendOption=Linear}
    boundBone {boneName=dyn_wing04_r}
    componentAppliedLinearAccClamp=<2000 2000 2000>
    componentLinearAccScale=<4 4 4>
    componentLinearVelScale=<4 4 4>
    gravityOverride {z=-980}
    physicsBodyDefinitions [{
      sphereCollisionRadius=0
      boundBone {boneName=dyn_wing04_r}
      boxExtents=<20 10 20>
      constraintSetup {angularLimitsMax=<1 0 3>, angularLimitsMin=<-1 0 -3>, angularTarget=<1 0 0>}
      localJointOffset=<0.1 0 0>
     }]
    relativeSpaceBone {boneName=pack}
   }
   Anim Dynamics - Bone: dyn_wing05_r(ComponentPose=@n1, Alpha=1, ExternalForce="0,0,0") {
    alpha=0.5
    angularDampingOverride=0.999
    angularSpringConstant=2000
    bAngularSpring=true
    bGravityOverrideInSimSpace=true
    bOverrideAngularDamping=true
    bUseGravityOverride=true
    simulationSpace=RootRelative
    alphaBoolBlend {blendOption=Linear}
    boundBone {boneName=dyn_wing05_r}
    componentAppliedLinearAccClamp=<2000 2000 2000>
    componentLinearAccScale=<4 4 4>
    componentLinearVelScale=<4 4 4>
    gravityOverride {z=-980}
    physicsBodyDefinitions [{
      sphereCollisionRadius=0
      boundBone {boneName=dyn_wing05_r}
      boxExtents=<10 10 10>
      constraintSetup {angularLimitsMax=<1 0 3>, angularLimitsMin=<-1 0 -3>, angularTarget=<1 0 0>}
      localJointOffset=<0.1 0 0>
     }]
    relativeSpaceBone {boneName=pack}
   }
   n2 = "Transform (Modify) Bone - Bone: root"!(ComponentPose=Local To Component(LocalPose=Input Pose()), bAlphaBoolEnabled:bool=True, Translation="0,-2.5,0", Rotation="0,0,0", Scale) -> (Pose) {
    alphaInputType=Bool
    translationMode=BMM_Additive
    translationSpace=BCS_BoneSpace
    alphaBoolBlend {blendOption=Linear}
    boneToModify {boneName=root}
    translation {y=-2.5}
   }
   n3 = "Transform (Modify) Bone - Bone: root"(ComponentPose=@n2, bAlphaBoolEnabled=True, Translation="0,2,0", Rotation="0,0,0") {
    alphaInputType=Bool
    translationMode=BMM_Additive
    translationSpace=BCS_BoneSpace
    alphaBoolBlend {blendInTime=0.1, blendOption=Sinusoidal, blendOutTime=0.1}
    boneToModify {boneName=root}
    translation {y=2}
   }
   n4 = Anim Dynamics - Bone: dyn_wing01_l(ComponentPose=@n3, Alpha=1, ExternalForce="0,0,0") {
    alpha=0.5
    angularDampingOverride=0.999
    angularSpringConstant=2000
    bAngularSpring=true
    bGravityOverrideInSimSpace=true
    bOverrideAngularDamping=true
    bUseGravityOverride=true
    simulationSpace=RootRelative
    alphaBoolBlend {blendOption=Linear}
    boundBone {boneName=dyn_wing01_l}
    componentAppliedLinearAccClamp=<2000 2000 2000>
    componentLinearAccScale=<4 4 4>
    componentLinearVelScale=<4 4 4>
    gravityOverride {z=-980}
    physicsBodyDefinitions [{
      sphereCollisionRadius=0
      boundBone {boneName=dyn_wing01_l}
      boxExtents=<20 10 20>
      constraintSetup {angularLimitsMax=<0 0 0>, angularLimitsMin=<0 0 -3>, angularTarget=<1 0 0>}
      localJointOffset=<0.3 0 0>
     }]
    relativeSpaceBone {boneName=pack}
   }
   n5 = Anim Dynamics - Bone: dyn_wing02_l(ComponentPose=@n4, Alpha=1, ExternalForce="0,0,0") {
    alpha=0.5
    angularDampingOverride=0.999
    angularSpringConstant=2000
    bAngularSpring=true
    bGravityOverrideInSimSpace=true
    bOverrideAngularDamping=true
    bUseGravityOverride=true
    simulationSpace=RootRelative
    alphaBoolBlend {blendOption=Linear}
    boundBone {boneName=dyn_wing02_l}
    componentAppliedLinearAccClamp=<2000 2000 2000>
    componentLinearAccScale=<4 4 4>
    componentLinearVelScale=<4 4 4>
    gravityOverride {z=-980}
    physicsBodyDefinitions [{
      sphereCollisionRadius=0
      boundBone {boneName=dyn_wing02_l}
      boxExtents=<20 10 20>
      constraintSetup {angularLimitsMax=<2 0 0>, angularLimitsMin=<-2 0 -3>, angularTarget=<1 0 0>}
      localJointOffset=<0.2 0 0>
     }]
    relativeSpaceBone {boneName=pack}
   }
   n6 = Anim Dynamics - Bone: dyn_wing03_l(ComponentPose=@n5, Alpha=1, ExternalForce="0,0,0") {
    alpha=0.5
    angularDampingOverride=0.999
    angularSpringConstant=2000
    bAngularSpring=true
    bGravityOverrideInSimSpace=true
    bOverrideAngularDamping=true
    bUseGravityOverride=true
    simulationSpace=RootRelative
    alphaBoolBlend {blendOption=Linear}
    boundBone {boneName=dyn_wing03_l}
    componentAppliedLinearAccClamp=<2000 2000 2000>
    componentLinearAccScale=<4 4 4>
    componentLinearVelScale=<4 4 4>
    gravityOverride {z=-980}
    physicsBodyDefinitions [{
      sphereCollisionRadius=0
      boundBone {boneName=dyn_wing03_l}
      boxExtents=<20 10 20>
      constraintSetup {angularLimitsMax=<1 0 3>, angularLimitsMin=<-1 0 -3>, angularTarget=<1 0 0>}
      localJointOffset=<-0.1 0 0>
     }]
    relativeSpaceBone {boneName=pack}
   }
   n7 = Anim Dynamics - Bone: dyn_wing04_l(ComponentPose=@n6, Alpha=1, ExternalForce="0,0,0") {
    alpha=0.5
    angularDampingOverride=0.999
    angularSpringConstant=2000
    bAngularSpring=true
    bGravityOverrideInSimSpace=true
    bOverrideAngularDamping=true
    bUseGravityOverride=true
    simulationSpace=RootRelative
    alphaBoolBlend {blendOption=Linear}
    boundBone {boneName=dyn_wing04_l}
    componentAppliedLinearAccClamp=<2000 2000 2000>
    componentLinearAccScale=<4 4 4>
    componentLinearVelScale=<4 4 4>
    gravityOverride {z=-980}
    physicsBodyDefinitions [{
      sphereCollisionRadius=0
      boundBone {boneName=dyn_wing04_l}
      boxExtents=<20 10 20>
      constraintSetup {angularLimitsMax=<1 0 3>, angularLimitsMin=<-1 0 -3>, angularTarget=<1 0 0>}
      localJointOffset=<-0.1 0 0>
     }]
    relativeSpaceBone {boneName=pack}
   }
   n8 = Anim Dynamics - Bone: dyn_wing05_l(ComponentPose=@n7, Alpha=1, ExternalForce="0,0,0") {
    alpha=0.5
    angularDampingOverride=0.999
    angularSpringConstant=2000
    bAngularSpring=true
    bGravityOverrideInSimSpace=true
    bOverrideAngularDamping=true
    bUseGravityOverride=true
    simulationSpace=RootRelative
    alphaBoolBlend {blendOption=Linear}
    boundBone {boneName=dyn_wing05_l}
    componentAppliedLinearAccClamp=<2000 2000 2000>
    componentLinearAccScale=<4 4 4>
    componentLinearVelScale=<4 4 4>
    gravityOverride {z=-980}
    physicsBodyDefinitions [{
      sphereCollisionRadius=0
      boundBone {boneName=dyn_wing05_l}
      boxExtents=<10 10 10>
      constraintSetup {angularLimitsMax=<1 0 3>, angularLimitsMin=<-1 0 -3>, angularTarget=<1 0 0>}
      localJointOffset=<-0.1 0 0>
     }]
    relativeSpaceBone {boneName=pack}
   }
   n9 = Anim Dynamics - Bone: dyn_wing01_r(ComponentPose=@n8, Alpha=1, ExternalForce="0,0,0") {
    alpha=0.5
    angularDampingOverride=0.999
    angularSpringConstant=2000
    bAngularSpring=true
    bGravityOverrideInSimSpace=true
    bOverrideAngularDamping=true
    bUseGravityOverride=true
    simulationSpace=RootRelative
    alphaBoolBlend {blendOption=Linear}
    boundBone {boneName=dyn_wing01_r}
    componentAppliedLinearAccClamp=<2000 2000 2000>
    componentLinearAccScale=<4 4 4>
    componentLinearVelScale=<4 4 4>
    gravityOverride {z=-980}
    physicsBodyDefinitions [{
      sphereCollisionRadius=0
      boundBone {boneName=dyn_wing01_r}
      boxExtents=<20 10 20>
      constraintSetup {angularLimitsMax=<0 0 0>, angularLimitsMin=<0 0 -3>, angularTarget=<1 0 0>}
      localJointOffset=<-0.3 0 0>
     }]
    relativeSpaceBone {boneName=pack}
   }
   n10 = Anim Dynamics - Bone: dyn_wing02_r(ComponentPose=@n9, Alpha=1, ExternalForce="0,0,0") {
    alpha=0.5
    angularDampingOverride=0.999
    angularSpringConstant=2000
    bAngularSpring=true
    bGravityOverrideInSimSpace=true
    bOverrideAngularDamping=true
    bUseGravityOverride=true
    simulationSpace=RootRelative
    alphaBoolBlend {blendOption=Linear}
    boundBone {boneName=dyn_wing02_r}
    componentAppliedLinearAccClamp=<2000 2000 2000>
    componentLinearAccScale=<4 4 4>
    componentLinearVelScale=<4 4 4>
    gravityOverride {z=-980}
    physicsBodyDefinitions [{
      sphereCollisionRadius=0
      boundBone {boneName=dyn_wing02_r}
      boxExtents=<20 10 20>
      constraintSetup {angularLimitsMax=<2 0 0>, angularLimitsMin=<-2 0 -3>, angularTarget=<1 0 0>}
      localJointOffset=<-0.2 0 0>
     }]
    relativeSpaceBone {boneName=pack}
   }
   n11 = Anim Dynamics - Bone: dyn_wing03_r(ComponentPose=@n10, Alpha=1, ExternalForce="0,0,0") {
    alpha=0.5
    angularDampingOverride=0.999
    angularSpringConstant=2000
    bAngularSpring=true
    bGravityOverrideInSimSpace=true
    bOverrideAngularDamping=true
    bUseGravityOverride=true
    simulationSpace=RootRelative
    alphaBoolBlend {blendOption=Linear}
    boundBone {boneName=dyn_wing03_r}
    componentAppliedLinearAccClamp=<2000 2000 2000>
    componentLinearAccScale=<4 4 4>
    componentLinearVelScale=<4 4 4>
    gravityOverride {z=-980}
    physicsBodyDefinitions [{
      sphereCollisionRadius=0
      boundBone {boneName=dyn_wing03_r}
      boxExtents=<20 10 20>
      constraintSetup {angularLimitsMax=<1 0 3>, angularLimitsMin=<-1 0 -3>, angularTarget=<1 0 0>}
      localJointOffset=<0.1 0 0>
     }]
    relativeSpaceBone {boneName=pack}
   }
   n12 = Anim Dynamics - Bone: feather09_upper_l(ComponentPose=@n11, Alpha=1, ExternalForce="0,0,0") {
    alpha=0.5
    angularDampingOverride=0.9999
    angularSpringConstant=4000
    bAngularSpring=true
    bGravityOverrideInSimSpace=true
    bOverrideAngularDamping=true
    bUseGravityOverride=true
    linearDampingOverride=0.8
    simulationSpace=RootRelative
    alphaBoolBlend {blendOption=Linear}
    boundBone {boneName=feather09_upper_l}
    componentAppliedLinearAccClamp=<2000 2000 2000>
    componentLinearAccScale=<3 3 3>
    componentLinearVelScale=<3 3 3>
    gravityOverride {z=-980}
    physicsBodyDefinitions [{
      sphereCollisionRadius=0
      boundBone {boneName=feather09_upper_l}
      boxExtents=<10 20 20>
      constraintSetup {angularLimitsMax=<0 5 2>, angularLimitsMin=<0 -5 -5>, angularTarget=<1 0 0>}
      localJointOffset=<0.3 0 0>
     }]
    relativeSpaceBone {boneName=pack}
   }
   n13 = Anim Dynamics - Bone: feather09_middle_l(ComponentPose=@n12, Alpha=1, ExternalForce="0,0,0") {
    alpha=0.5
    angularDampingOverride=1
    angularSpringConstant=8000
    bAngularSpring=true
    bGravityOverrideInSimSpace=true
    bOverrideAngularDamping=true
    bUseGravityOverride=true
    linearDampingOverride=0.8
    simulationSpace=RootRelative
    alphaBoolBlend {blendOption=Linear}
    boundBone {boneName=feather09_middle_l}
    componentAppliedLinearAccClamp=<1500 1500 1500>
    componentLinearAccScale=<3 3 3>
    componentLinearVelScale=<3 3 3>
    gravityOverride {z=-980}
    physicsBodyDefinitions [{
      sphereCollisionRadius=0
      boundBone {boneName=feather09_middle_l}
      boxExtents=<10 20 20>
      constraintSetup {angularLimitsMax=<0 10 0>, angularLimitsMin=<0 -10 0>, angularTarget=<1 0 0>}
      localJointOffset=<0.3 0 0>
     }]
    relativeSpaceBone {boneName=pack}
   }
   n14 = Anim Dynamics - Bone: feather09_bottom_l(ComponentPose=@n13, Alpha=1, ExternalForce="0,0,0") {
    alpha=0.5
    angularDampingOverride=1
    angularSpringConstant=8000
    bAngularSpring=true
    bGravityOverrideInSimSpace=true
    bOverrideAngularDamping=true
    bUseGravityOverride=true
    linearDampingOverride=0.8
    simulationSpace=RootRelative
    alphaBoolBlend {blendOption=Linear}
    boundBone {boneName=feather09_bottom_l}
    componentAppliedLinearAccClamp=<1500 1500 1500>
    componentLinearAccScale=<3 3 3>
    componentLinearVelScale=<3 3 3>
    gravityOverride {z=-980}
    physicsBodyDefinitions [{
      sphereCollisionRadius=0
      boundBone {boneName=feather09_bottom_l}
      boxExtents=<10 20 20>
      constraintSetup {angularLimitsMax=<0 15 0>, angularLimitsMin=<0 -15 0>, angularTarget=<1 0 0>}
      localJointOffset=<0.3 0 0>
     }]
    relativeSpaceBone {boneName=pack}
   }
   n15 = Anim Dynamics - Bone: feather08_upper_l(ComponentPose=@n14, Alpha=1, ExternalForce="0,0,0") {
    alpha=0.5
    angularDampingOverride=0.9999
    angularSpringConstant=4000
    bAngularSpring=true
    bGravityOverrideInSimSpace=true
    bOverrideAngularDamping=true
    bUseGravityOverride=true
    linearDampingOverride=0.8
    simulationSpace=RootRelative
    alphaBoolBlend {blendOption=Linear}
    boundBone {boneName=feather08_upper_l}
    componentAppliedLinearAccClamp=<2000 2000 2000>
    componentLinearAccScale=<3 3 3>
    componentLinearVelScale=<3 3 3>
    gravityOverride {z=-980}
    physicsBodyDefinitions [{
      sphereCollisionRadius=0
      boundBone {boneName=feather08_upper_l}
      boxExtents=<10 20 20>
      constraintSetup {angularLimitsMax=<0 5 2>, angularLimitsMin=<0 -5 -5>, angularTarget=<1 0 0>}
      localJointOffset=<0.3 0 0>
     }]
    relativeSpaceBone {boneName=pack}
   }
   n16 = Anim Dynamics - Bone: feather08_middle_l(ComponentPose=@n15, Alpha=1, ExternalForce="0,0,0") {
    alpha=0.5
    angularDampingOverride=1
    angularSpringConstant=8000
    bAngularSpring=true
    bGravityOverrideInSimSpace=true
    bOverrideAngularDamping=true
    bUseGravityOverride=true
    linearDampingOverride=0.8
    simulationSpace=RootRelative
    alphaBoolBlend {blendOption=Linear}
    boundBone {boneName=feather08_middle_l}
    componentAppliedLinearAccClamp=<1500 1500 1500>
    componentLinearAccScale=<3 3 3>
    componentLinearVelScale=<3 3 3>
    gravityOverride {z=-980}
    physicsBodyDefinitions [{
      sphereCollisionRadius=0
      boundBone {boneName=feather08_middle_l}
      boxExtents=<10 20 20>
      constraintSetup {angularLimitsMax=<0 10 0>, angularLimitsMin=<0 -10 0>, angularTarget=<1 0 0>}
      localJointOffset=<0.3 0 0>
     }]
    relativeSpaceBone {boneName=pack}
   }
   n17 = Anim Dynamics - Bone: feather08_bottom_l(ComponentPose=@n16, Alpha=1, ExternalForce="0,0,0") {
    alpha=0.5
    angularDampingOverride=1
    angularSpringConstant=8000
    bAngularSpring=true
    bGravityOverrideInSimSpace=true
    bOverrideAngularDamping=true
    bUseGravityOverride=true
    linearDampingOverride=0.8
    simulationSpace=RootRelative
    alphaBoolBlend {blendOption=Linear}
    boundBone {boneName=feather08_bottom_l}
    componentAppliedLinearAccClamp=<1500 1500 1500>
    componentLinearAccScale=<3 3 3>
    componentLinearVelScale=<3 3 3>
    gravityOverride {z=-980}
    physicsBodyDefinitions [{
      sphereCollisionRadius=0
      boundBone {boneName=feather08_bottom_l}
      boxExtents=<10 20 20>
      constraintSetup {angularLimitsMax=<0 15 0>, angularLimitsMin=<0 -15 0>, angularTarget=<1 0 0>}
      localJointOffset=<0.3 0 0>
     }]
    relativeSpaceBone {boneName=pack}
   }
   n18 = Anim Dynamics - Bone: feather07_upper_l(ComponentPose=@n17, Alpha=1, ExternalForce="0,0,0") {
    alpha=0.5
    angularDampingOverride=0.9999
    angularSpringConstant=4000
    bAngularSpring=true
    bGravityOverrideInSimSpace=true
    bOverrideAngularDamping=true
    bUseGravityOverride=true
    linearDampingOverride=0.8
    simulationSpace=RootRelative
    alphaBoolBlend {blendOption=Linear}
    boundBone {boneName=feather07_upper_l}
    componentAppliedLinearAccClamp=<2000 2000 2000>
    componentLinearAccScale=<3 3 3>
    componentLinearVelScale=<3 3 3>
    gravityOverride {z=-980}
    physicsBodyDefinitions [{
      sphereCollisionRadius=0
      boundBone {boneName=feather07_upper_l}
      boxExtents=<10 20 20>
      constraintSetup {angularLimitsMax=<0 5 1>, angularLimitsMin=<0 -5 -2>, angularTarget=<1 0 0>}
      localJointOffset=<0.15 0 0>
     }]
    relativeSpaceBone {boneName=pack}
   }
   n19 = Anim Dynamics - Bone: feather07_middle_l(ComponentPose=@n18, Alpha=1, ExternalForce="0,0,0") {
    alpha=0.5
    angularDampingOverride=1
    angularSpringConstant=8000
    bAngularSpring=true
    bGravityOverrideInSimSpace=true
    bOverrideAngularDamping=true
    bUseGravityOverride=true
    linearDampingOverride=0.8
    simulationSpace=RootRelative
    alphaBoolBlend {blendOption=Linear}
    boundBone {boneName=feather07_middle_l}
    componentAppliedLinearAccClamp=<1500 1500 1500>
    componentLinearAccScale=<3 3 3>
    componentLinearVelScale=<3 3 3>
    gravityOverride {z=-980}
    physicsBodyDefinitions [{
      sphereCollisionRadius=0
      boundBone {boneName=feather07_middle_l}
      boxExtents=<10 20 20>
      constraintSetup {angularLimitsMax=<0 10 0>, angularLimitsMin=<0 -10 0>, angularTarget=<1 0 0>}
      localJointOffset=<0.15 0 0>
     }]
    relativeSpaceBone {boneName=pack}
   }
   n20 = Anim Dynamics - Bone: feather07_bottom_l(ComponentPose=@n19, Alpha=1, ExternalForce="0,0,0") {
    alpha=0.5
    angularDampingOverride=1
    angularSpringConstant=8000
    bAngularSpring=true
    bGravityOverrideInSimSpace=true
    bOverrideAngularDamping=true
    bUseGravityOverride=true
    linearDampingOverride=0.8
    simulationSpace=RootRelative
    alphaBoolBlend {blendOption=Linear}
    boundBone {boneName=feather07_bottom_l}
    componentAppliedLinearAccClamp=<1500 1500 1500>
    componentLinearAccScale=<3 3 3>
    componentLinearVelScale=<3 3 3>
    gravityOverride {z=-980}
    physicsBodyDefinitions [{
      sphereCollisionRadius=0
      boundBone {boneName=feather07_bottom_l}
      boxExtents=<10 20 20>
      constraintSetup {angularLimitsMax=<0 15 0>, angularLimitsMin=<0 -15 0>, angularTarget=<1 0 0>}
      localJointOffset=<0.15 0 0>
     }]
    relativeSpaceBone {boneName=pack}
   }
   n21 = Anim Dynamics - Bone: feather06_upper_l(ComponentPose=@n20, Alpha=1, ExternalForce="0,0,0") {
    alpha=0.5
    angularDampingOverride=0.9999
    angularSpringConstant=4000
    bAngularSpring=true
    bGravityOverrideInSimSpace=true
    bOverrideAngularDamping=true
    bUseGravityOverride=true
    linearDampingOverride=0.8
    simulationSpace=RootRelative
    alphaBoolBlend {blendOption=Linear}
    boundBone {boneName=feather06_upper_l}
    componentAppliedLinearAccClamp=<2000 2000 2000>
    componentLinearAccScale=<3 3 3>
    componentLinearVelScale=<3 3 3>
    gravityOverride {z=-980}
    physicsBodyDefinitions [{
      sphereCollisionRadius=0
      boundBone {boneName=feather06_upper_l}
      boxExtents=<10 20 20>
      constraintSetup {angularLimitsMax=<0 5 0>, angularLimitsMin=<0 -5 0>, angularTarget=<1 0 0>}
      localJointOffset=<0.1 0 0>
     }]
    relativeSpaceBone {boneName=feather07_upper_l}
   }
   n22 = Anim Dynamics - Bone: feather06_middle_l(ComponentPose=@n21, Alpha=1, ExternalForce="0,0,0") {
    alpha=0.5
    angularDampingOverride=1
    angularSpringConstant=8000
    bAngularSpring=true
    bGravityOverrideInSimSpace=true
    bOverrideAngularDamping=true
    bUseGravityOverride=true
    linearDampingOverride=0.8
    simulationSpace=RootRelative
    alphaBoolBlend {blendOption=Linear}
    boundBone {boneName=feather06_middle_l}
    componentAppliedLinearAccClamp=<1500 1500 1500>
    componentLinearAccScale=<3 3 3>
    componentLinearVelScale=<3 3 3>
    gravityOverride {z=-980}
    physicsBodyDefinitions [{
      sphereCollisionRadius=0
      boundBone {boneName=feather06_middle_l}
      boxExtents=<10 20 20>
      constraintSetup {angularLimitsMax=<0 10 0>, angularLimitsMin=<0 -10 0>, angularTarget=<1 0 0>}
      localJointOffset=<0.1 0 0>
     }]
    relativeSpaceBone {boneName=pack}
   }
   n23 = Anim Dynamics - Bone: feather06_bottom_l(ComponentPose=@n22, Alpha=1, ExternalForce="0,0,0") {
    alpha=0.5
    angularDampingOverride=1
    angularSpringConstant=8000
    bAngularSpring=true
    bGravityOverrideInSimSpace=true
    bOverrideAngularDamping=true
    bUseGravityOverride=true
    linearDampingOverride=0.8
    simulationSpace=RootRelative
    alphaBoolBlend {blendOption=Linear}
    boundBone {boneName=feather06_bottom_l}
    componentAppliedLinearAccClamp=<1500 1500 1500>
    componentLinearAccScale=<3 3 3>
    componentLinearVelScale=<3 3 3>
    gravityOverride {z=-980}
    physicsBodyDefinitions [{
      sphereCollisionRadius=0
      boundBone {boneName=feather06_bottom_l}
      boxExtents=<10 20 20>
      constraintSetup {angularLimitsMax=<0 15 0>, angularLimitsMin=<0 -15 0>, angularTarget=<1 0 0>}
      localJointOffset=<0.1 0 0>
     }]
    relativeSpaceBone {boneName=pack}
   }
   n24 = Anim Dynamics - Bone: feather09_upper_r(ComponentPose=@n23, Alpha=1, ExternalForce="0,0,0") {
    alpha=0.5
    angularDampingOverride=0.9999
    angularSpringConstant=4000
    bAngularSpring=true
    bGravityOverrideInSimSpace=true
    bOverrideAngularDamping=true
    bUseGravityOverride=true
    linearDampingOverride=0.8
    simulationSpace=RootRelative
    alphaBoolBlend {blendOption=Linear}
    boundBone {boneName=feather09_upper_r}
    componentAppliedLinearAccClamp=<2000 2000 2000>
    componentLinearAccScale=<3 3 3>
    componentLinearVelScale=<3 3 3>
    gravityOverride {z=-980}
    physicsBodyDefinitions [{
      sphereCollisionRadius=0
      boundBone {boneName=feather09_upper_r}
      boxExtents=<10 20 20>
      constraintSetup {angularLimitsMax=<0 5 2>, angularLimitsMin=<0 -5 -5>, angularTarget=<1 0 0>}
      localJointOffset=<-0.3 0 0>
     }]
    relativeSpaceBone {boneName=pack}
   }
   n25 = Anim Dynamics - Bone: feather09_middle_r(ComponentPose=@n24, Alpha=1, ExternalForce="0,0,0") {
    alpha=0.5
    angularDampingOverride=1
    angularSpringConstant=8000
    bAngularSpring=true
    bGravityOverrideInSimSpace=true
    bOverrideAngularDamping=true
    bUseGravityOverride=true
    linearDampingOverride=0.8
    simulationSpace=RootRelative
    alphaBoolBlend {blendOption=Linear}
    boundBone {boneName=feather09_middle_r}
    componentAppliedLinearAccClamp=<1500 1500 1500>
    componentLinearAccScale=<3 3 3>
    componentLinearVelScale=<3 3 3>
    gravityOverride {z=-980}
    physicsBodyDefinitions [{
      sphereCollisionRadius=0
      boundBone {boneName=feather09_middle_r}
      boxExtents=<10 20 20>
      constraintSetup {angularLimitsMax=<0 10 0>, angularLimitsMin=<0 -10 0>, angularTarget=<1 0 0>}
      localJointOffset=<-0.3 0 0>
     }]
    relativeSpaceBone {boneName=pack}
   }
   n26 = Anim Dynamics - Bone: feather09_bottom_r(ComponentPose=@n25, Alpha=1, ExternalForce="0,0,0") {
    alpha=0.5
    angularDampingOverride=1
    angularSpringConstant=8000
    bAngularSpring=true
    bGravityOverrideInSimSpace=true
    bOverrideAngularDamping=true
    bUseGravityOverride=true
    linearDampingOverride=0.8
    simulationSpace=RootRelative
    alphaBoolBlend {blendOption=Linear}
    boundBone {boneName=feather09_bottom_r}
    componentAppliedLinearAccClamp=<1500 1500 1500>
    componentLinearAccScale=<3 3 3>
    componentLinearVelScale=<3 3 3>
    gravityOverride {z=-980}
    physicsBodyDefinitions [{
      sphereCollisionRadius=0
      boundBone {boneName=feather09_bottom_r}
      boxExtents=<10 20 20>
      constraintSetup {angularLimitsMax=<0 15 0>, angularLimitsMin=<0 -15 0>, angularTarget=<1 0 0>}
      localJointOffset=<-0.3 0 0>
     }]
    relativeSpaceBone {boneName=pack}
   }
   n27 = Anim Dynamics - Bone: feather08_upper_r(ComponentPose=@n26, Alpha=1, ExternalForce="0,0,0") {
    alpha=0.5
    angularDampingOverride=0.9999
    angularSpringConstant=4000
    bAngularSpring=true
    bGravityOverrideInSimSpace=true
    bOverrideAngularDamping=true
    bUseGravityOverride=true
    linearDampingOverride=0.8
    simulationSpace=RootRelative
    alphaBoolBlend {blendOption=Linear}
    boundBone {boneName=feather08_upper_r}
    componentAppliedLinearAccClamp=<2000 2000 2000>
    componentLinearAccScale=<3 3 3>
    componentLinearVelScale=<3 3 3>
    gravityOverride {z=-980}
    physicsBodyDefinitions [{
      sphereCollisionRadius=0
      boundBone {boneName=feather08_upper_r}
      boxExtents=<10 20 20>
      constraintSetup {angularLimitsMax=<0 5 2>, angularLimitsMin=<0 -5 -5>, angularTarget=<1 0 0>}
      localJointOffset=<-0.3 0 0>
     }]
    relativeSpaceBone {boneName=pack}
   }
   n28 = Anim Dynamics - Bone: feather08_middle_r(ComponentPose=@n27, Alpha=1, ExternalForce="0,0,0") {
    alpha=0.5
    angularDampingOverride=1
    angularSpringConstant=8000
    bAngularSpring=true
    bGravityOverrideInSimSpace=true
    bOverrideAngularDamping=true
    bUseGravityOverride=true
    linearDampingOverride=0.8
    simulationSpace=RootRelative
    alphaBoolBlend {blendOption=Linear}
    boundBone {boneName=feather08_middle_r}
    componentAppliedLinearAccClamp=<1500 1500 1500>
    componentLinearAccScale=<3 3 3>
    componentLinearVelScale=<3 3 3>
    gravityOverride {z=-980}
    physicsBodyDefinitions [{
      sphereCollisionRadius=0
      boundBone {boneName=feather08_middle_r}
      boxExtents=<10 20 20>
      constraintSetup {angularLimitsMax=<0 10 0>, angularLimitsMin=<0 -10 0>, angularTarget=<1 0 0>}
      localJointOffset=<-0.3 0 0>
     }]
    relativeSpaceBone {boneName=pack}
   }
   n29 = Anim Dynamics - Bone: feather08_bottom_r(ComponentPose=@n28, Alpha=1, ExternalForce="0,0,0") {
    alpha=0.5
    angularDampingOverride=1
    angularSpringConstant=8000
    bAngularSpring=true
    bGravityOverrideInSimSpace=true
    bOverrideAngularDamping=true
    bUseGravityOverride=true
    linearDampingOverride=0.8
    simulationSpace=RootRelative
    alphaBoolBlend {blendOption=Linear}
    boundBone {boneName=feather08_bottom_r}
    componentAppliedLinearAccClamp=<1500 1500 1500>
    componentLinearAccScale=<3 3 3>
    componentLinearVelScale=<3 3 3>
    gravityOverride {z=-980}
    physicsBodyDefinitions [{
      sphereCollisionRadius=0
      boundBone {boneName=feather08_bottom_r}
      boxExtents=<10 20 20>
      constraintSetup {angularLimitsMax=<0 15 0>, angularLimitsMin=<0 -15 0>, angularTarget=<1 0 0>}
      localJointOffset=<-0.3 0 0>
     }]
    relativeSpaceBone {boneName=pack}
   }
   n30 = Anim Dynamics - Bone: feather07_upper_r(ComponentPose=@n29, Alpha=1, ExternalForce="0,0,0") {
    alpha=0.5
    angularDampingOverride=0.9999
    angularSpringConstant=4000
    bAngularSpring=true
    bGravityOverrideInSimSpace=true
    bOverrideAngularDamping=true
    bUseGravityOverride=true
    linearDampingOverride=0.8
    simulationSpace=RootRelative
    alphaBoolBlend {blendOption=Linear}
    boundBone {boneName=feather07_upper_r}
    componentAppliedLinearAccClamp=<2000 2000 2000>
    componentLinearAccScale=<3 3 3>
    componentLinearVelScale=<3 3 3>
    gravityOverride {z=-980}
    physicsBodyDefinitions [{
      sphereCollisionRadius=0
      boundBone {boneName=feather07_upper_r}
      boxExtents=<10 20 20>
      constraintSetup {angularLimitsMax=<0 5 1>, angularLimitsMin=<0 -5 -2>, angularTarget=<1 0 0>}
      localJointOffset=<-0.15 0 0>
     }]
    relativeSpaceBone {boneName=pack}
   }
   n31 = Anim Dynamics - Bone: feather07_middle_r(ComponentPose=@n30, Alpha=1, ExternalForce="0,0,0") {
    alpha=0.5
    angularDampingOverride=1
    angularSpringConstant=8000
    bAngularSpring=true
    bGravityOverrideInSimSpace=true
    bOverrideAngularDamping=true
    bUseGravityOverride=true
    linearDampingOverride=0.8
    simulationSpace=RootRelative
    alphaBoolBlend {blendOption=Linear}
    boundBone {boneName=feather07_middle_r}
    componentAppliedLinearAccClamp=<1500 1500 1500>
    componentLinearAccScale=<3 3 3>
    componentLinearVelScale=<3 3 3>
    gravityOverride {z=-980}
    physicsBodyDefinitions [{
      sphereCollisionRadius=0
      boundBone {boneName=feather07_middle_r}
      boxExtents=<10 20 20>
      constraintSetup {angularLimitsMax=<0 10 0>, angularLimitsMin=<0 -10 0>, angularTarget=<1 0 0>}
      localJointOffset=<-0.15 0 0>
     }]
    relativeSpaceBone {boneName=pack}
   }
   n32 = Anim Dynamics - Bone: feather07_bottom_r(ComponentPose=@n31, Alpha=1, ExternalForce="0,0,0") {
    alpha=0.5
    angularDampingOverride=1
    angularSpringConstant=8000
    bAngularSpring=true
    bGravityOverrideInSimSpace=true
    bOverrideAngularDamping=true
    bUseGravityOverride=true
    linearDampingOverride=0.8
    simulationSpace=RootRelative
    alphaBoolBlend {blendOption=Linear}
    boundBone {boneName=feather07_bottom_r}
    componentAppliedLinearAccClamp=<1500 1500 1500>
    componentLinearAccScale=<3 3 3>
    componentLinearVelScale=<3 3 3>
    gravityOverride {z=-980}
    physicsBodyDefinitions [{
      sphereCollisionRadius=0
      boundBone {boneName=feather07_bottom_r}
      boxExtents=<10 20 20>
      constraintSetup {angularLimitsMax=<0 15 0>, angularLimitsMin=<0 -15 0>, angularTarget=<1 0 0>}
      localJointOffset=<-0.15 0 0>
     }]
    relativeSpaceBone {boneName=pack}
   }
   n33 = Anim Dynamics - Bone: feather06_upper_r(ComponentPose=@n32, Alpha=1, ExternalForce="0,0,0") {
    alpha=0.5
    angularDampingOverride=0.9999
    angularSpringConstant=4000
    bAngularSpring=true
    bGravityOverrideInSimSpace=true
    bOverrideAngularDamping=true
    bUseGravityOverride=true
    linearDampingOverride=0.8
    simulationSpace=RootRelative
    alphaBoolBlend {blendOption=Linear}
    boundBone {boneName=feather06_upper_r}
    componentAppliedLinearAccClamp=<2000 2000 2000>
    componentLinearAccScale=<3 3 3>
    componentLinearVelScale=<3 3 3>
    gravityOverride {z=-980}
    physicsBodyDefinitions [{
      sphereCollisionRadius=0
      boundBone {boneName=feather06_upper_r}
      boxExtents=<10 20 20>
      constraintSetup {angularLimitsMax=<0 5 0>, angularLimitsMin=<0 -5 0>, angularTarget=<1 0 0>}
      localJointOffset=<-0.1 0 0>
     }]
    relativeSpaceBone {boneName=feather07_upper_r}
   }
   n34 = Anim Dynamics - Bone: feather06_middle_r(ComponentPose=@n33, Alpha=1, ExternalForce="0,0,0") {
    alpha=0.5
    angularDampingOverride=1
    angularSpringConstant=8000
    bAngularSpring=true
    bGravityOverrideInSimSpace=true
    bOverrideAngularDamping=true
    bUseGravityOverride=true
    linearDampingOverride=0.8
    simulationSpace=RootRelative
    alphaBoolBlend {blendOption=Linear}
    boundBone {boneName=feather06_middle_r}
    componentAppliedLinearAccClamp=<1500 1500 1500>
    componentLinearAccScale=<3 3 3>
    componentLinearVelScale=<3 3 3>
    gravityOverride {z=-980}
    physicsBodyDefinitions [{
      sphereCollisionRadius=0
      boundBone {boneName=feather06_middle_r}
      boxExtents=<10 20 20>
      constraintSetup {angularLimitsMax=<0 10 0>, angularLimitsMin=<0 -10 0>, angularTarget=<1 0 0>}
      localJointOffset=<-0.1 0 0>
     }]
    relativeSpaceBone {boneName=pack}
   }
   n35 = Anim Dynamics - Bone: feather06_bottom_r(ComponentPose=@n34, Alpha=1, ExternalForce="0,0,0") {
    alpha=0.5
    angularDampingOverride=1
    angularSpringConstant=8000
    bAngularSpring=true
    bGravityOverrideInSimSpace=true
    bOverrideAngularDamping=true
    bUseGravityOverride=true
    linearDampingOverride=0.8
    simulationSpace=RootRelative
    alphaBoolBlend {blendOption=Linear}
    boundBone {boneName=feather06_bottom_r}
    componentAppliedLinearAccClamp=<1500 1500 1500>
    componentLinearAccScale=<3 3 3>
    componentLinearVelScale=<3 3 3>
    gravityOverride {z=-980}
    physicsBodyDefinitions [{
      sphereCollisionRadius=0
      boundBone {boneName=feather06_bottom_r}
      boxExtents=<10 20 20>
      constraintSetup {angularLimitsMax=<0 15 0>, angularLimitsMin=<0 -15 0>, angularTarget=<1 0 0>}
      localJointOffset=<-0.1 0 0>
     }]
    relativeSpaceBone {boneName=pack}
   }
   n36 = "Transform (Modify) Bone - Bone: root"(ComponentPose=@n35, bAlphaBoolEnabled=True, Translation="0,-4,0", Rotation="0,0,30", Scale="0.01,0.01,0.01") {
    alphaInputType=Bool
    rotationSpace=BCS_BoneSpace
    scaleMode=BMM_Replace
    scaleSpace=BCS_BoneSpace
    translationSpace=BCS_BoneSpace
    alphaBoolBlend {blendOption=Linear}
    alphaScaleBiasClamp {bInterpResult=true, interpSpeedDecreasing=5, interpSpeedIncreasing=5}
    boneToModify {boneName=root}
    rotation {roll=30}
    scale=<0.01 0.01 0.01>
    translation {y=-4}
   }
   n37 = Control Rig!(
    Source=Component To Local(ComponentPose=@n36)
    Alpha
    SkydiveDiveAlpha:double=Get SkydiveDiveModeAlpha() -> (SkydiveDiveModeAlpha:double)
    SkydiveEnterAlpha:double=Get SkydiveEnterAlpha() -> (SkydiveEnterAlpha:double)
    TacticalSprintingAlpha:double=Get TacticalSprintingAlpha() -> (TacticalSprintingAlpha:double)
    GrindingSprintAlpha:double=Get GrindingSprintAlpha() -> (GrindingSprintAlpha:double)
    GrindingAlpha:double=Get GrindingAlpha() -> (GrindingAlpha:double)
    SlidingAlpha:double=Get SlidingAlpha() -> (SlidingAlpha:double)
    SkydiveFloatingAlpha:double=Get SkydiveFloatingAlpha() -> (SkydiveFloatingAlpha:double)
    JumpOrFalling:double=Get JumpOrFallingAlpha() -> (JumpOrFallingAlpha:double)
    isCrouching:double=Get CrouchingAlpha() -> (CrouchingAlpha:double)
    isTargeting:double=Get TargetingAlpha() -> (TargetingAlpha:double)
    LeftRightSideMovement:double=Get Player Vector LR() -> (Player Vector LR:double)
    ForwardBackwardMovement:double=Get Player Vector Fwd Bwd() -> (Player Vector Fwd Bwd:double)
    SpineSwimmingAlpha:double=Get SpineSwimmingAlpha() -> (SpineSwimmingAlpha:float)
    HeadRotation:double=Get TurnInTwistAlpha() -> (TurnInTwistAlpha:float)
   ) -> (Pose) {
    alphaBoolBlend {blendOption=Linear}
    controlRigAssetReference {blueprintRigClass=&/F_MED_CoatCheck_Pack_CtrlRig.F_MED_CoatCheck_Pack_CtrlRig_C}
   }
   Output Pose(Result=@n37) {layerGroup=None}
  ]}
 EventGraph {nodes [
   Event BlueprintUpdateAnimation()
   > n1 = PlayerMovement(AnimInstance=Self-Reference(), XDotProductCoefficient=1, YDotProductCoefficient=1, ZDotProductCoefficient=1)
   > Set Player Vector LR(Player Vector LR:double=@n1) -> (Output_Get:double)
   > Set Player Vector Fwd Bwd(Player Vector Fwd Bwd:double=@n1.YMovement) -> (Output_Get:double)
   > Set SkydiveDiveModeAlpha(SkydiveDiveModeAlpha:double="To Float (Boolean)"(InBool=Get bIsSkydiveDiveMode())) -> (Output_Get:double)
   > Set bIsMaleCrouching(bIsMaleCrouching:bool=AND Boolean(A=Get bIsMale() -> (bIsMale:bool), B=Get bIsCrouching())) -> (Output_Get:bool)
   n2 = AND Boolean(A=NOT Boolean(A=Get bIsCrouching()), B=NOT Boolean(A=Get bIsSliding()))
   n3 = AND Boolean!(A=Get bIsTacticalSprinting(), B=NOT Boolean(A=Get bIsJumping()), C:bool=@n2) -> (ReturnValue)
   n4 = "To Float (Boolean)"(InBool=AND Boolean!(A=Get bIsSprinting(), B=@n2) -> (ReturnValue))
   > Set TacticalSprintingAlpha(TacticalSprintingAlpha:double=float + float(A="To Float (Boolean)"(InBool=@n3), B=float * float(A=@n4, B=0.5) {is=title K2Node_PromotableOperator})) -> (Output_Get:double)
   > Set GrindingAlpha(GrindingAlpha:double="To Float (Boolean)"(InBool=Get bIsOnGrindRail())) -> (Output_Get:double)
   > Set SlidingAlpha(SlidingAlpha:double="To Float (Boolean)"(InBool=Get bIsSliding())) -> (Output_Get:double)
   > Set SkydiveFloatingAlpha(SkydiveFloatingAlpha:double="To Float (Boolean)"(InBool=Get bIsSkydiveFloating())) -> (Output_Get:double)
   n5 = NOT Boolean(A=Get bIsParachuteOpened())
   n6 = OR Boolean(A=Get bIsJumping(), B=Get bIsFalling())
   > Set JumpOrFallingAlpha(JumpOrFallingAlpha:double="To Float (Boolean)"(InBool=AND Boolean(A=@n6, B=@n5))) -> (Output_Get:double)
   n7 = NOT Boolean(A=Get bIsSliding())
   n8 = Get bIsCrouching()
   > Set CrouchingAlpha(CrouchingAlpha:double="To Float (Boolean)"(InBool=AND Boolean(A=@n8, B=@n7))) -> (Output_Get:double)
   > Set TargetingAlpha(TargetingAlpha:double="To Float (Boolean)"(InBool=Get bIsTargeting())) -> (Output_Get:double)
   n9 = GetOwnerSkeletalMeshComponent()
   n10 = Get bIsSurfaceSwimming()
   > Set SpineSwimmingAlpha(SpineSwimmingAlpha:float=float * float(
     A=GetAlphaFromReferenceTransformDelta(SkeletalMeshComponent=@n9, BoneName=spine_05, BaseName=spine_01, InputRange=(X=-30,Y=0), OutputRange=(X=1,Y=0), TwistAxis="1,0,0")
     B="To Float (Boolean)"(InBool=@n10)
    )) -> (Output_Get:float)
   n11 = GetTurnInPlaceInput!(self:Fort Player Animation Interface Object Reference=GetOwnerAnimationInterface!(self:Custom Character Part Anim Instance Object Reference) -> (ReturnValue:Fort Player Animation Interface Object Reference)) -> (ReturnValue:"Fort Anim Input Turn In Place Structure (by ref)")
   > Set TurnInTwistAlpha(TurnInTwistAlpha:float=Remap Float(
     Input=Break Fort Anim Input Turn in Place!(FortAnimInput_TurnInPlace:"Fort Anim Input Turn In Place Structure (by ref)"=@n11) -> (
      RootYawOffset:float=~
      TurnAnimPlayRate:float=~
      bWantsToTurn:bool=~
      bExitTurn:bool=~
      bUseTurnA:bool=~
      TurnACurveAlpha:float=~
      TurnBCurveAlpha:float=~
      bWantsToTurnInVehicle:bool=~
      bWantsToTurnAgain:bool=~
      bTurningLeft:bool=~
      ShufflePlayRate:float=~
      CrouchShufflePlayRate:float=~
      CrouchShuffleStartPosition:float=~
      CrouchShuffleSpeedScaling:float=~
      bTransition_LowerBody_Shuffle:bool=~
      bTransition_LowerBody_ShuffleADS:bool=~
      bPlayOutro:bool=~
     ) {is=title K2Node_BreakStruct}
     OldRange=(X=-35,Y=-75)
     NewRange=(X=0,Y=1)
     Clamp=true
    ) -> (Output=~)) -> (Output_Get:float)
   > n12 = Branch(Condition=GliderItemDefinitionHasAnyMetaTag!(Tags:Array of Gameplay Tag Structures=Make Array("[0]"=(TagName="Cosmetics.MetaTags.CoatCheck")), bRequiresOpened:bool=true) -> (ReturnValue:bool))
   > Set bIsCoatCheckGliderDeploy(bIsCoatCheckGliderDeploy:bool=true) -> (Output_Get:bool)
   > Delay(Duration=0.2)
   > Set PackHiddenAlpha(PackHiddenAlpha:bool=1) -> (Output_Get:bool)
   Set bIsCoatCheckGliderDeploy(execute=@n12.else, bIsCoatCheckGliderDeploy=false)
   > Set PackHiddenAlpha(PackHiddenAlpha=false)
   Event OnBeginSkydiving()
   > Set SkydiveEnterAlpha(SkydiveEnterAlpha:double=1) -> (Output_Get:double)
   > Delay(Duration=1)
   > Set SkydiveEnterAlpha(SkydiveEnterAlpha=0)
   n13 = Event OnCharacterCustomizationLoaded()
   Set bIsMale(
    execute=[@n13.then, @n14.then]
    bIsMale:bool="Equal (Enum)"(
     A:EFortCustomGender Enum=GetOwnerGender!(self:Custom Character Part Anim Instance Object Reference) -> (ReturnValue:EFortCustomGender Enum=Invalid)
     B:EFortCustomGender Enum=Male
    )
   ) -> (Output_Get:bool)
   n14 = Event BlueprintInitializeAnimation()
   Delay(Duration=0.1)
  ]}
]
variables [
 bIsCoatCheckGliderDeploy {type=boolean, default=true}
 bIsMale {type=boolean}
 bIsMaleCrouching {type=boolean}
 CrouchingAlpha {type=number}
 GrindingAlpha {type=number}
 GrindingSprintAlpha {type=number}
 JumpOrFallingAlpha {type=number}
 PackHiddenAlpha {type=boolean}
 Player Vector Fwd Bwd {type=number}
 Player Vector LR {type=number}
 SkydiveDiveModeAlpha {type=number}
 SkydiveEnterAlpha {type=number}
 SkydiveFloatingAlpha {type=number}
 SlidingAlpha {type=number}
 SpineSwimmingAlpha {type=number}
 TacticalSprintingAlpha {type=number}
 TargetingAlpha {type=number}
 TurnInTwistAlpha {type=number}
]
